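# Niagara system for the Keychain character skin that wraps music-reactive ribbons around the player's arm and hand bones.
# Separate frontend and gameplay emitters, toggled by a spawn count tied to bIsFrontEndPreview, each burst-spawn points that FN_BoneFlow places along filtered skeletal bones.
# Each feeds two ribbon renderers with curve-shaped widths, colors and dynamic material parameters.
# System values drive time-of-day brightness, WPO and alpha, with distance fading in gameplay.
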NiagaraSystem /BRCosmetics/Characters/Player/Male/Medium/Bodies/M_MED_Keychain/FX/NS_Keychain_Skin
dependencies {
 usedDataInterfaces [
  &NiagaraDataInterfaceCurve
  &NiagaraDataInterfaceCamera
  &NiagaraDataInterfacePlatformSet
  &NiagaraDataInterfaceSkeletalMesh
 ]
 usedDynamicInputs [
  &FNTimeOfDayScalar
  &Waveform
  &RandomRangeFloat
  &SetBoolByFloatComparison
  &Power
 ]
 usedModules [
  &self:SystemSpawnScript.SystemScriptSource.SystemScriptGraph.NiagaraNodeAssignment_4.SetVariables_#1
  &self:SystemSpawnScript.SystemScriptSource.SystemScriptGraph.NiagaraNodeAssignment_5.SetVariables_#2
  &self:SystemSpawnScript.SystemScriptSource.SystemScriptGraph.NiagaraNodeAssignment_6.SetVariables_#3
  &SystemState
  &self:SystemSpawnScript.SystemScriptSource.SystemScriptGraph.NiagaraNodeAssignment_0.SetVariables_#4
  &self:SystemSpawnScript.SystemScriptSource.SystemScriptGraph.NiagaraNodeAssignment_1.SetVariables_#5
  &self:SystemSpawnScript.SystemScriptSource.SystemScriptGraph.NiagaraNodeAssignment_7.SetVariables_#6
  &self:SystemSpawnScript.SystemScriptSource.SystemScriptGraph.NiagaraNodeAssignment_2.SetVariables_#7
  &self:SystemSpawnScript.SystemScriptSource.SystemScriptGraph.NiagaraNodeAssignment_3.SetVariables_#8
  &self:SystemSpawnScript.SystemScriptSource.SystemScriptGraph.NiagaraNodeAssignment_9.SetVariables_#9
  &InterpolateOverTime
 ]
}
emitters {
 Bone_Flow_L_Hand_Frontend {
  emitterUpdateScript [
   &EmitterState(
    Inactive Response=::NewEnumerator1
    Life Cycle Mode=::NewEnumerator1
    Loop Behavior=::NewEnumerator1
    Loop Duration Mode=::NewEnumerator1
    UseLoopDelay=false
   )
   &SpawnBurst_Instantaneous(
    Spawn Count=&MakeCustomIntFromBool(Bool=$User.bIsFrontEndPreview, True INT=10, False INT=0)
    Use Spawn Probability=false
    UseLoopCountLimit=false
   )
  ]
  particleSpawnScript [
   &FN_SetSpawnGroup(Inclusivity=true, Normalize Spawn Group=true, Particles Per Group=10, Pattern=::NewEnumerator0, Set Ribbon ID and Ribbon Link Order=true)
   &InitializeParticle(
    Color=<rgba 2.2 2.2 2.2 1>
    Ribbon Width=10
    Ribbon Width Mode=::NewEnumerator1
    Sprite Size Mode=::NewEnumerator3
    Uniform Sprite Size=1
    UsePositionOffset=false
    Write Material Random=true
   )
   &CameraOffset(Camera Offset Amount=30)
  ]
  particleUpdateScript [
   /SetVariables(Particles.Ribbon width02:NiagaraFloat=&Multiply_Float(
     A=15
     B=&FloatFromCurve(
      FloatCurve=DataInterface() {propertyValues {Curve {keys [
          {interpMode=RCIM_Cubic, tangentMode=RCTM_User, time=0, value=0.3}
          {interpMode=RCIM_Cubic, tangentMode=RCTM_User, time=0.3, value=1}
          {arriveTangent=-1, interpMode=RCIM_Cubic, leaveTangent=-1, time=0.7, value=1}
          {tangentMode=RCTM_User, time=1, value=0.3}
         ]}}}
      CurveIndex=$Particles.Normalized Spawn Group
     )
    )) {isSetParameters=true, module=SetVariables_#10}
   &ParticleState(Kill Particles When Lifetime Has Elapsed=false, Let Infinitely Lived Particles Die When Emitter Deactivates=true)
   &FN_BoneFlow(
    Angle:NiagaraFloat=&Multiply_Float(A=$Particles.Normalized Spawn Group, B=0.4)
    BoneID:NiagaraInt32=&ModuloInt(A=$Particles.SpawnGroup, B=1)
    Height:NiagaraFloat=&Multiply_Float(A=$Particles.Normalized Spawn Group, B=0.6)
    Optimize Rotation Stabalization=true
    Radius:NiagaraFloat=&FloatFromCurve(
     FloatCurve=DataInterface() {propertyValues {Curve {keys [
         {interpMode=RCIM_Cubic, tangentMode=RCTM_User, time=0, value=0.4}
         {interpMode=RCIM_Cubic, tangentMode=RCTM_User, time=0.5, value=1}
         {arriveTangent=-0.3574, tangentMode=RCTM_User, time=1, value=0.4}
        ]}}}
     CurveIndex=$Particles.Normalized Spawn Group
     Scale Curve=10
    )
    SideVector2:Vector3f=<0 0 0>
    SideVector3:Vector3f=<0 -2 0>
    SkeletalMesh=DataInterface() {propertyValues {
      DefaultMesh=null
      ExcludeBoneName=root
      OptionalComponentClass=null
      PreviewMesh=&M_Med_Soldier_04_Preview
      SkinningMode=SkinOnTheFly
      SoftSourceActor=null
      SourceMode=Default
      UvSetIndex=0
      WholeMeshLOD=-1
      bExcludeBone=false
      bReadDeformedGeometry=true
      bRequireCurrentFrameData=true
      ComponentTags []
      FilteredBones [hand_l,lowerarm_l,upperarm_l]
      FilteredSockets []
      MeshUserParameter {parameter {name=None, typeDefHandle {registeredTypeIndex=1}, varData []}}
      SamplingRegions []
     }}
    Stabalize Twisting=true
   )
   &ScaleRibbonWidth(Ribbon Width Scale=&FloatFromCurve(
     FloatCurve=DataInterface() {propertyValues {Curve {keys [
         {interpMode=RCIM_Cubic, tangentMode=RCTM_User, time=0, value=0.5}
         {interpMode=RCIM_Cubic, tangentMode=RCTM_User, time=0.3, value=1}
         {arriveTangent=-0.7143, interpMode=RCIM_Cubic, leaveTangent=-0.7143, time=0.7, value=1}
         {tangentMode=RCTM_User, time=1, value=0.5}
        ]}}}
     CurveIndex=$Particles.Normalized Spawn Group
    ))
   /SetVariables(Particles.Ribbon02 Color:LinearColor=&MakeLinearColorFromVectorAndFloat(
     "Vector (RGB)"=&MakeVectorFromLinearColorRGB(Color=&MultiplyLinearColor(Linear Color=<rgba 1.5 1.5 1.5 1>, Scale Factor=&MakeVector4_FromFloat(Float=$System.TOD)))
     "Float (Alpha)"=&Multiply_Float(A=&Multiply_Float(A=1.5, B=1), B=$System.Frontend Alpha NoMusic)
    )) {isSetParameters=true, module=SetVariables_#11}
   /SetVariables(Particles.Position=&AddVectorToPosition(
     Position=$Particles.Position
     Vector=&Multiply_VectorByFloat(
      Vector=&NormalizeVector(
       Vector To Normalize=&SubtractVectorFromPosition(
        Position=&ReturnCameraPosition(Coordinate Space:ENiagaraCoordinateSpace=::Simulation, Camera Query:NiagaraDataInterfaceCamera=DataInterface())
        Vector=Linked() {linkedVariable {name=Particles.Position, type=NiagaraPosition}}
       )
       Safe Normalize=true
       Safe Fallback Threshold=0
       Safe Fallback Vector=<0 0 0.001>
      )
      Float=0
     )
    )) {enabled=false, isSetParameters=true, module=SetVariables_#12}
   &DynamicMaterialParameters(
    Index 0 Param 1=$System.WPO
    Index 0 Param 2=&Lerp_Float(A=0.7, B=1.1, Alpha=$System.Music_Reactive_preview)
    Module.Index0Param0Write=true
    Module.Index0Param1Write=true
    Module.Index0Param2Write=true
    Module.Index0Param3Write=true
   )
   &ScaleColor(
    Scale Alpha=$System.Frontend Alpha NoMusic
    Scale RGB=&Multiply_VectorByFloat(
     Vector=&Multiply_VectorByFloat(
      Vector=<1 1 1>
      Float=&FN_PickFloatDependOnPlatform(TargetPlatformActive=1, TargetPlatformNoActive=0.75)
     )
     Float=$System.TOD
    )
    ScaleA=true
    ScaleRGB=true
   )
  ]
  rendererClasses [&NiagaraRibbonRendererProperties]
  renderers [
   {rendererClass=&NiagaraRibbonRendererProperties}
   {rendererClass=&NiagaraRibbonRendererProperties, rendererIndex=1}
  ]
 }
 Bone_Flow_L_UpperArm02_Gameplay {
  emitterUpdateScript [
   &EmitterState(
    Inactive Response=::NewEnumerator1
    Life Cycle Mode=::NewEnumerator1
    Loop Behavior=::NewEnumerator1
    Loop Duration Mode=::NewEnumerator1
    UseLoopDelay=false
   )
   &SpawnBurst_Instantaneous(
    Spawn Count=&MakeCustomIntFromBool(Bool=$User.bIsFrontEndPreview, True INT=0, False INT=10)
    Use Spawn Probability=false
    UseLoopCountLimit=false
   )
  ]
  particleSpawnScript [
   &FN_SetSpawnGroup(Inclusivity=true, Normalize Spawn Group=true, Particles Per Group=10, Pattern=::NewEnumerator0, Set Ribbon ID and Ribbon Link Order=true)
   &InitializeParticle(
    Color=<rgba 2.2 2.2 2.2 1>
    Ribbon Width=10
    Ribbon Width Mode=::NewEnumerator1
    Sprite Size Mode=::NewEnumerator3
    Uniform Sprite Size=1
    UsePositionOffset=false
    Write Material Random=true
   )
   &CameraOffset(Camera Offset Amount=30)
  ]
  particleUpdateScript [
   /SetVariables(Particles.Ribbon width02:NiagaraFloat=&Multiply_Float(
     A=15
     B=&FloatFromCurve(
      FloatCurve=DataInterface() {propertyValues {Curve {keys [
          {interpMode=RCIM_Cubic, tangentMode=RCTM_User, time=0, value=0.3}
          {interpMode=RCIM_Cubic, tangentMode=RCTM_User, time=0.3, value=1}
          {arriveTangent=-1, interpMode=RCIM_Cubic, leaveTangent=-1, time=0.7, value=1}
          {tangentMode=RCTM_User, time=1, value=0.3}
         ]}}}
      CurveIndex=$Particles.Normalized Spawn Group
     )
    )) {isSetParameters=true, module=SetVariables_#10}
   &ParticleState(Kill Particles When Lifetime Has Elapsed=false, Let Infinitely Lived Particles Die When Emitter Deactivates=true)
   &FN_BoneFlow(
    Angle:NiagaraFloat=&Multiply_Float(A=$Particles.Normalized Spawn Group, B=0.3)
    BoneID:NiagaraInt32=&ModuloInt(A=$Particles.SpawnGroup, B=1)
    Enable Position Offsets=true
    Height:NiagaraFloat=&Multiply_Float(A=$Particles.Normalized Spawn Group, B=1)
    Offset Mid:Vector3f=<0 0 5>
    Offset Start:Vector3f=<0 5 0>
    Optimize Rotation Stabalization=true
    Radius:NiagaraFloat=&FloatFromCurve(
     FloatCurve=DataInterface() {propertyValues {Curve {keys [
         {interpMode=RCIM_Cubic, tangentMode=RCTM_User, time=0, value=0.4}
         {interpMode=RCIM_Cubic, tangentMode=RCTM_User, time=0.5, value=1.5}
         {arriveTangent=-0.3574, tangentMode=RCTM_User, time=1, value=0.4}
        ]}}}
     CurveIndex=$Particles.Normalized Spawn Group
     Scale Curve=5
    )
    SideVector1:Vector3f=<-1 0 -2>
    SideVector2:Vector3f=<0 0 0>
    SideVector3:Vector3f=<0 -1 0>
    SkeletalMesh=DataInterface() {propertyValues {
      DefaultMesh=null
      ExcludeBoneName=root
      OptionalComponentClass=null
      PreviewMesh=&M_Med_Soldier_04_Preview
      SkinningMode=SkinOnTheFly
      SoftSourceActor=null
      SourceMode=Default
      UvSetIndex=0
      WholeMeshLOD=-1
      bExcludeBone=false
      bReadDeformedGeometry=true
      bRequireCurrentFrameData=true
      ComponentTags []
      FilteredBones [lowerarm_l,upperarm_l,clavicle_l]
      FilteredSockets []
      MeshUserParameter {parameter {name=None, typeDefHandle {registeredTypeIndex=1}, varData []}}
      SamplingRegions []
     }}
    Stabalize Twisting=true
   )
   &ScaleRibbonWidth(Ribbon Width Scale=&FloatFromCurve(
     FloatCurve=DataInterface() {propertyValues {Curve {keys [
         {interpMode=RCIM_Cubic, tangentMode=RCTM_User, time=0, value=0.5}
         {interpMode=RCIM_Cubic, tangentMode=RCTM_User, time=0.3, value=1}
         {arriveTangent=-0.7143, interpMode=RCIM_Cubic, leaveTangent=-0.7143, time=0.7, value=1}
         {tangentMode=RCTM_User, time=1, value=0.5}
        ]}}}
     CurveIndex=$Particles.Normalized Spawn Group
    ))
   /SetVariables(Particles.Ribbon02 Color:LinearColor=&MakeLinearColorFromVectorAndFloat(
     "Vector (RGB)"=&MakeVectorFromLinearColorRGB(Color=&MultiplyLinearColor(Linear Color=<rgba 1.5 1.5 1.5 1>, Scale Factor=&MakeVector4_FromFloat(Float=$System.TOD)))
     "Float (Alpha)"=&Multiply_Float(
      A=&Multiply_Float(
       A=&Multiply_Float(
        A=1.5
        B=&Lerp_Float(A=0, B=1, Alpha=&ClampFloat(Float=$System.Music_Reactive_preview))
       )
       B=&Multiply_Float(
        A=$User.Opacity
        B=&FN_FrontEndFloatModifier(
         Gameplay=&FN_LOD_Distance_Remap(
          Start Distance=200
          End Distance=300
          Minimum Output=&Lerp_Float(A=0.1, B=1, Alpha=$System.LOD_Fade_preview)
          Maximum Output=1
          Clamp Output=true
         )
         Front End=1
        )
       )
      )
      B=&FloatFromCurve(
       FloatCurve=DataInterface() {propertyValues {Curve {keys [
           {interpMode=RCIM_Cubic, tangentMode=RCTM_User, time=0, value=0}
           {interpMode=RCIM_Cubic, tangentMode=RCTM_User, time=0.3, value=1}
           {tangentMode=RCTM_User, time=1, value=1}
          ]}}}
       CurveIndex=$Particles.Normalized Spawn Group
      )
     )
    )) {isSetParameters=true, module=SetVariables_#11}
   /SetVariables(Particles.Position=&AddVectorToPosition(
     Position=$Particles.Position
     Vector=&Multiply_VectorByFloat(
      Vector=&NormalizeVector(
       Vector To Normalize=&SubtractVectorFromPosition(
        Position=&ReturnCameraPosition(Coordinate Space:ENiagaraCoordinateSpace=::Simulation, Camera Query:NiagaraDataInterfaceCamera=DataInterface())
        Vector=Linked() {linkedVariable {name=Particles.Position, type=NiagaraPosition}}
       )
       Safe Normalize=true
       Safe Fallback Threshold=0
       Safe Fallback Vector=<0 0 0.001>
      )
      Float=-10
     )
    )) {isSetParameters=true, module=SetVariables_#12}
   &DynamicMaterialParameters(
    Index 0 Param 1=&Multiply_Float(A=$System.WPO, B=0.8)
    Index 0 Param 2=&Lerp_Float(A=0.7, B=1.1, Alpha=$System.Music_Reactive_preview)
    Module.Index0Param0Write=true
    Module.Index0Param1Write=true
    Module.Index0Param2Write=true
    Module.Index0Param3Write=true
   )
   &ScaleColor(
    Scale Alpha=&Multiply_Float(
     A=&Multiply_Float(
      A=&Lerp_Float(A=0, B=1, Alpha=$System.InterpolateOverTime.InterpolateOverTime_Float.Moving Average)
      B=&Multiply_Float(
       A=$User.Opacity
       B=&FN_FrontEndFloatModifier(
        Gameplay=&FN_LOD_Distance_Remap(
         Start Distance=200
         End Distance=300
         Minimum Output=&Lerp_Float(A=0.01, B=1, Alpha=$System.LOD_Fade_preview)
         Maximum Output=1
         Clamp Output=true
        )
        Front End=1
       )
      )
     )
     B=&FloatFromCurve(
      FloatCurve=DataInterface() {propertyValues {Curve {keys [
          {interpMode=RCIM_Cubic, tangentMode=RCTM_User, time=0.3, value=0}
          {interpMode=RCIM_Cubic, tangentMode=RCTM_User, time=0.4, value=1}
          {tangentMode=RCTM_User, time=1, value=1}
         ]}}}
      CurveIndex=$Particles.Normalized Spawn Group
     )
    )
    Scale RGB=&Multiply_VectorByFloat(
     Vector=&Multiply_VectorByFloat(
      Vector=<1 1 1>
      Float=&FN_PickFloatDependOnPlatform(TargetPlatformActive=1, TargetPlatformNoActive=0.75)
     )
     Float=$System.TOD
    )
    ScaleA=true
    ScaleRGB=true
   )
  ]
  rendererClasses [&NiagaraRibbonRendererProperties]
  renderers [
   {rendererClass=&NiagaraRibbonRendererProperties}
   {rendererClass=&NiagaraRibbonRendererProperties, rendererIndex=1}
  ]
 }
 Bone_Flow_L_UpperArm_Frontend {
  emitterUpdateScript [
   &EmitterState(
    Inactive Response=::NewEnumerator1
    Life Cycle Mode=::NewEnumerator1
    Loop Behavior=::NewEnumerator1
    Loop Duration Mode=::NewEnumerator1
    UseLoopDelay=false
   )
   &SpawnBurst_Instantaneous(
    Spawn Count=&MakeCustomIntFromBool(Bool=$User.bIsFrontEndPreview, True INT=10, False INT=0)
    Use Spawn Probability=false
    UseLoopCountLimit=false
   )
  ]
  particleSpawnScript [
   &FN_SetSpawnGroup(Inclusivity=true, Normalize Spawn Group=true, Particles Per Group=10, Pattern=::NewEnumerator0, Set Ribbon ID and Ribbon Link Order=true)
   &InitializeParticle(
    Color=<rgba 2.2 2.2 2.2 1>
    Ribbon Width=10
    Ribbon Width Mode=::NewEnumerator1
    Sprite Size Mode=::NewEnumerator3
    Uniform Sprite Size=1
    UsePositionOffset=false
    Write Material Random=true
   )
   &CameraOffset(Camera Offset Amount=30)
  ]
  particleUpdateScript [
   /SetVariables(Particles.Ribbon width02:NiagaraFloat=&Multiply_Float(
     A=15
     B=&FloatFromCurve(
      FloatCurve=DataInterface() {propertyValues {Curve {keys [
          {interpMode=RCIM_Cubic, tangentMode=RCTM_User, time=0, value=0.3}
          {interpMode=RCIM_Cubic, tangentMode=RCTM_User, time=0.3, value=1}
          {arriveTangent=-1, interpMode=RCIM_Cubic, leaveTangent=-1, time=0.7, value=1}
          {tangentMode=RCTM_User, time=1, value=0.3}
         ]}}}
      CurveIndex=$Particles.Normalized Spawn Group
     )
    )) {isSetParameters=true, module=SetVariables_#10}
   &ParticleState(Kill Particles When Lifetime Has Elapsed=false, Let Infinitely Lived Particles Die When Emitter Deactivates=true)
   &FN_BoneFlow(
    Angle:NiagaraFloat=&Multiply_Float(A=$Particles.Normalized Spawn Group, B=&FN_FrontEndFloatModifier(Gameplay=0.5, Front End=0.8))
    BoneID:NiagaraInt32=&ModuloInt(A=$Particles.SpawnGroup, B=1)
    Enable Position Offsets=true
    Height:NiagaraFloat=&Multiply_Float(A=$Particles.Normalized Spawn Group, B=1)
    Offset Start:Vector3f=<5 5 0>
    Optimize Rotation Stabalization=true
    Radius:NiagaraFloat=&FloatFromCurve(
     FloatCurve=DataInterface() {propertyValues {Curve {keys [
         {interpMode=RCIM_Cubic, tangentMode=RCTM_User, time=0, value=0.4}
         {interpMode=RCIM_Cubic, tangentMode=RCTM_User, time=0.5, value=1}
         {arriveTangent=-0.3574, tangentMode=RCTM_User, time=1, value=0.4}
        ]}}}
     CurveIndex=$Particles.Normalized Spawn Group
     Scale Curve=10
    )
    SideVector1:Vector3f=<0 -2 -1>
    SideVector2:Vector3f=<0 0 0>
    SideVector3:Vector3f=<0 0 0>
    SkeletalMesh=DataInterface() {propertyValues {
      DefaultMesh=null
      ExcludeBoneName=root
      OptionalComponentClass=null
      PreviewMesh=&M_Med_Soldier_04_Preview
      SkinningMode=SkinOnTheFly
      SoftSourceActor=null
      SourceMode=Default
      UvSetIndex=0
      WholeMeshLOD=-1
      bExcludeBone=false
      bReadDeformedGeometry=true
      bRequireCurrentFrameData=true
      ComponentTags []
      FilteredBones [lowerarm_l,upperarm_l,clavicle_l]
      FilteredSockets []
      MeshUserParameter {parameter {name=None, typeDefHandle {registeredTypeIndex=1}, varData []}}
      SamplingRegions []
     }}
    Stabalize Twisting=true
   )
   &ScaleRibbonWidth(Ribbon Width Scale=&FloatFromCurve(
     FloatCurve=DataInterface() {propertyValues {Curve {keys [
         {interpMode=RCIM_Cubic, tangentMode=RCTM_User, time=0, value=0.5}
         {interpMode=RCIM_Cubic, tangentMode=RCTM_User, time=0.3, value=1}
         {arriveTangent=-0.7143, interpMode=RCIM_Cubic, leaveTangent=-0.7143, time=0.7, value=1}
         {tangentMode=RCTM_User, time=1, value=0.5}
        ]}}}
     CurveIndex=$Particles.Normalized Spawn Group
    ))
   /SetVariables(Particles.Ribbon02 Color:LinearColor=&MakeLinearColorFromVectorAndFloat(
     "Vector (RGB)"=&MakeVectorFromLinearColorRGB(Color=&MultiplyLinearColor(Linear Color=<rgba 1.5 1.5 1.5 1>, Scale Factor=&MakeVector4_FromFloat(Float=$System.TOD)))
     "Float (Alpha)"=&Multiply_Float(
      A=&Multiply_Float(A=1.5, B=1)
      B=&FloatFromCurve(
       FloatCurve=DataInterface() {propertyValues {Curve {keys [<at 0.1 0>,<at 0.3 1>,<at 1 1>]}}}
       CurveIndex=$StackContext.RibbonLinkOrder
       Scale Curve=$System.Frontend Alpha NoMusic
      )
     )
    )) {isSetParameters=true, module=SetVariables_#11}
   /SetVariables(Particles.Position=&AddVectorToPosition(
     Position=$Particles.Position
     Vector=&Multiply_VectorByFloat(
      Vector=&NormalizeVector(
       Vector To Normalize=&SubtractVectorFromPosition(
        Position=&ReturnCameraPosition(Coordinate Space:ENiagaraCoordinateSpace=::Simulation, Camera Query:NiagaraDataInterfaceCamera=DataInterface())
        Vector=Linked() {linkedVariable {name=Particles.Position, type=NiagaraPosition}}
       )
       Safe Normalize=true
       Safe Fallback Threshold=0
       Safe Fallback Vector=<0 0 0.001>
      )
      Float=0
     )
    )) {enabled=false, isSetParameters=true, module=SetVariables_#12}
   &DynamicMaterialParameters(
    Index 0 Param 1=$System.WPO
    Index 0 Param 2=&Lerp_Float(A=0.7, B=1.1, Alpha=$System.Music_Reactive_preview)
    Module.Index0Param0Write=true
    Module.Index0Param1Write=true
    Module.Index0Param2Write=true
    Module.Index0Param3Write=true
   )
   &ScaleColor(
    Scale Alpha=&FloatFromCurve(
     FloatCurve=DataInterface() {propertyValues {Curve {keys [<at 0.1 0>,<at 0.3 1>,<at 1 1>]}}}
     CurveIndex=$StackContext.RibbonLinkOrder
     Scale Curve=$System.Frontend Alpha NoMusic
    )
    Scale RGB=&Multiply_VectorByFloat(
     Vector=&Multiply_VectorByFloat(
      Vector=<1 1 1>
      Float=&FN_PickFloatDependOnPlatform(TargetPlatformActive=1, TargetPlatformNoActive=0.75)
     )
     Float=$System.TOD
    )
    ScaleA=true
    ScaleRGB=true
   )
  ]
  rendererClasses [&NiagaraRibbonRendererProperties]
  renderers [
   {rendererClass=&NiagaraRibbonRendererProperties}
   {rendererClass=&NiagaraRibbonRendererProperties, rendererIndex=1}
  ]
 }
 Bone_Flow_L_Upperarm_GamePlay {
  emitterUpdateScript [
   &EmitterState(
    Inactive Response=::NewEnumerator1
    Life Cycle Mode=::NewEnumerator1
    Loop Behavior=::NewEnumerator1
    Loop Duration Mode=::NewEnumerator1
    UseLoopDelay=false
   )
   &SpawnBurst_Instantaneous(
    Spawn Count=&MakeCustomIntFromBool(Bool=$User.bIsFrontEndPreview, True INT=0, False INT=10)
    Use Spawn Probability=false
    UseLoopCountLimit=false
   )
  ]
  particleSpawnScript [
   &FN_SetSpawnGroup(Inclusivity=true, Normalize Spawn Group=true, Particles Per Group=10, Pattern=::NewEnumerator0, Set Ribbon ID and Ribbon Link Order=true)
   &InitializeParticle(
    Color=<rgba 2.2 2.2 2.2 1>
    Ribbon Width=10
    Ribbon Width Mode=::NewEnumerator1
    Sprite Size Mode=::NewEnumerator3
    Uniform Sprite Size=1
    UsePositionOffset=false
    Write Material Random=true
   )
   &CameraOffset(Camera Offset Amount=30)
  ]
  particleUpdateScript [
   &ParticleState(Kill Particles When Lifetime Has Elapsed=false, Let Infinitely Lived Particles Die When Emitter Deactivates=true)
   &FN_BoneFlow(
    Angle:NiagaraFloat=&Multiply_Float(A=$Particles.Normalized Spawn Group, B=&FN_FrontEndFloatModifier(Gameplay=0.5, Front End=0.8))
    BoneID:NiagaraInt32=&ModuloInt(A=$Particles.SpawnGroup, B=1)
    Height:NiagaraFloat=&Multiply_Float(A=$Particles.Normalized Spawn Group, B=1)
    Radius:NiagaraFloat=&FloatFromCurve(
     FloatCurve=DataInterface() {propertyValues {Curve {keys [
         {interpMode=RCIM_Cubic, tangentMode=RCTM_User, time=0, value=0.4}
         {interpMode=RCIM_Cubic, tangentMode=RCTM_User, time=0.5, value=1.5}
         {tangentMode=RCTM_User, time=1, value=0.4}
        ]}}}
     CurveIndex=$Particles.Normalized Spawn Group
     Scale Curve=10
    )
    SkeletalMesh:NiagaraDataInterfaceSkeletalMesh=$System.Skeletal Mesh L
   )
   &ScaleRibbonWidth(Ribbon Width Scale=&FloatFromCurve(
     FloatCurve=DataInterface() {propertyValues {Curve {keys [
         {interpMode=RCIM_Cubic, tangentMode=RCTM_User, time=0, value=0.5}
         {interpMode=RCIM_Cubic, tangentMode=RCTM_User, time=0.3, value=1}
         {arriveTangent=-0.7143, interpMode=RCIM_Cubic, leaveTangent=-0.7143, time=0.7, value=1}
         {tangentMode=RCTM_User, time=1, value=0.5}
        ]}}}
     CurveIndex=$Particles.Normalized Spawn Group
     Scale Curve=&MakeCustomFloatFromBool(Bool=$User.bisHandHidden, True Float=0, False Float=1)
    ))
   /SetVariables(Particles.Ribbon width02:NiagaraFloat=&Multiply_Float(
     A=15
     B=&FloatFromCurve(
      FloatCurve=DataInterface() {propertyValues {Curve {keys [
          {interpMode=RCIM_Cubic, tangentMode=RCTM_User, time=0, value=0.3}
          {interpMode=RCIM_Cubic, tangentMode=RCTM_User, time=0.3, value=1}
          {arriveTangent=-1, interpMode=RCIM_Cubic, leaveTangent=-1, time=0.7, value=1}
          {tangentMode=RCTM_User, time=1, value=0.3}
         ]}}}
      CurveIndex=$Particles.Normalized Spawn Group
      Scale Curve=&MakeCustomFloatFromBool(Bool=$User.bisHandHidden, True Float=0, False Float=1)
     )
    )) {isSetParameters=true, module=SetVariables_#10}
   /SetVariables(Particles.Ribbon02 Color:LinearColor=&MakeLinearColorFromVectorAndFloat(
     "Vector (RGB)"=&MakeVectorFromLinearColorRGB(Color=&MultiplyLinearColor(Linear Color=<rgba 1.5 1.5 1.5 1>, Scale Factor=&MakeVector4_FromFloat(Float=$System.TOD)))
     "Float (Alpha)"=&Multiply_Float(
      A=&Multiply_Float(
       A=1.5
       B=&Lerp_Float(A=0, B=1, Alpha=&ClampFloat(Float=$System.Music_Reactive_preview))
      )
      B=&Multiply_Float(
       A=$User.Opacity
       B=&FN_FrontEndFloatModifier(
        Gameplay=&FN_LOD_Distance_Remap(
         Start Distance=200
         End Distance=300
         Minimum Output=&Lerp_Float(A=0.1, B=1, Alpha=$System.LOD_Fade_preview)
         Maximum Output=1
         Clamp Output=true
        )
        Front End=1
       )
      )
     )
    )) {isSetParameters=true, module=SetVariables_#11}
   /SetVariables(Particles.Position=&AddVectorToPosition(
     Position=$Particles.Position
     Vector=&Multiply_VectorByFloat(
      Vector=&NormalizeVector(
       Vector To Normalize=&SubtractVectorFromPosition(
        Position=&ReturnCameraPosition(Coordinate Space:ENiagaraCoordinateSpace=::Simulation, Camera Query:NiagaraDataInterfaceCamera=DataInterface())
        Vector=Linked() {linkedVariable {name=Particles.Position, type=NiagaraPosition}}
       )
       Safe Normalize=true
       Safe Fallback Threshold=0
       Safe Fallback Vector=<0 0 0.001>
      )
      Float=-10
     )
    )) {isSetParameters=true, module=SetVariables_#12}
   &DynamicMaterialParameters(
    Index 0 Param 1=$System.WPO
    Index 0 Param 2=&Lerp_Float(A=0.7, B=1.1, Alpha=$System.Music_Reactive_preview)
    Module.Index0Param0Write=true
    Module.Index0Param1Write=true
    Module.Index0Param2Write=true
    Module.Index0Param3Write=true
   )
   &ScaleColor(
    Scale Alpha=&Multiply_Float(
     A=&Lerp_Float(A=0, B=1, Alpha=$System.InterpolateOverTime.InterpolateOverTime_Float.Moving Average)
     B=&Multiply_Float(
      A=$User.Opacity
      B=&FN_FrontEndFloatModifier(
       Gameplay=&FN_LOD_Distance_Remap(
        Start Distance=200
        End Distance=300
        Minimum Output=&Lerp_Float(A=0.01, B=1, Alpha=$System.LOD_Fade_preview)
        Maximum Output=1
        Clamp Output=true
       )
       Front End=1
      )
     )
    )
    Scale RGB=&Multiply_VectorByFloat(
     Vector=&Multiply_VectorByFloat(
      Vector=<1 1 1>
      Float=&FN_PickFloatDependOnPlatform(TargetPlatformActive=1, TargetPlatformNoActive=0.75)
     )
     Float=$System.TOD
    )
    ScaleA=true
    ScaleRGB=true
   )
  ]
  rendererClasses [&NiagaraRibbonRendererProperties]
  renderers [
   {rendererClass=&NiagaraRibbonRendererProperties}
   {rendererClass=&NiagaraRibbonRendererProperties, rendererIndex=1}
  ]
 }
 Bone_Flow_R_Hand_Frontend {
  emitterUpdateScript [
   &EmitterState(
    Inactive Response=::NewEnumerator1
    Life Cycle Mode=::NewEnumerator1
    Loop Behavior=::NewEnumerator1
    Loop Duration Mode=::NewEnumerator1
    UseLoopDelay=false
   )
   &SpawnBurst_Instantaneous(
    Spawn Count=&MakeCustomIntFromBool(Bool=$User.bIsFrontEndPreview, True INT=10, False INT=0)
    Use Spawn Probability=false
    UseLoopCountLimit=false
   )
  ]
  particleSpawnScript [
   &FN_SetSpawnGroup(Inclusivity=true, Normalize Spawn Group=true, Particles Per Group=10, Pattern=::NewEnumerator0, Set Ribbon ID and Ribbon Link Order=true)
   &InitializeParticle(
    Color=<rgba 2.2 2.2 2.2 1>
    Ribbon Width=10
    Ribbon Width Mode=::NewEnumerator1
    Sprite Size Mode=::NewEnumerator3
    Uniform Sprite Size=1
    UsePositionOffset=false
    Write Material Random=true
   )
   &CameraOffset(Camera Offset Amount=30)
  ]
  particleUpdateScript [
   /SetVariables(Particles.Ribbon width02:NiagaraFloat=&Multiply_Float(
     A=15
     B=&FloatFromCurve(
      FloatCurve=DataInterface() {propertyValues {Curve {keys [
          {interpMode=RCIM_Cubic, tangentMode=RCTM_User, time=0, value=0.3}
          {interpMode=RCIM_Cubic, tangentMode=RCTM_User, time=0.3, value=0.8}
          {arriveTangent=-0.7143, interpMode=RCIM_Cubic, leaveTangent=-0.7143, time=0.7, value=0.8}
          {tangentMode=RCTM_User, time=1, value=0.3}
         ]}}}
      CurveIndex=$Particles.Normalized Spawn Group
     )
    )) {isSetParameters=true, module=SetVariables_#10}
   &ParticleState(Kill Particles When Lifetime Has Elapsed=false, Let Infinitely Lived Particles Die When Emitter Deactivates=true)
   &FN_BoneFlow(
    Angle:NiagaraFloat=&FN_FrontEndFloatModifier(Gameplay=&Multiply_Float(A=$Particles.Normalized Spawn Group, B=0.3), Front End=0.2)
    BoneID:NiagaraInt32=&ModuloInt(A=$Particles.SpawnGroup, B=1)
    Enable Position Offsets=true
    Height:NiagaraFloat=&Multiply_Float(A=$Particles.Normalized Spawn Group, B=&FN_FrontEndFloatModifier(Gameplay=0.7, Front End=0.55))
    Offset Start:Vector3f=<-10 0 0>
    Optimize Rotation Stabalization=true
    Radius:NiagaraFloat=&FloatFromCurve(
     FloatCurve=DataInterface() {propertyValues {Curve {keys [
         {arriveTangent=1.3929, interpMode=RCIM_Cubic, leaveTangent=1.3929, tangentMode=RCTM_User, time=0, value=0}
         {interpMode=RCIM_Cubic, tangentMode=RCTM_User, time=0.5, value=1.3}
         {arriveTangent=-2.2123, tangentMode=RCTM_User, time=1, value=0.3}
        ]}}}
     CurveIndex=$Particles.Normalized Spawn Group
     Scale Curve=10
    )
    SideVector1:Vector3f=<0 0 1>
    SideVector2:Vector3f=<0 0 0>
    SkeletalMesh:NiagaraDataInterfaceSkeletalMesh=$System.Skeletal Mesh RHand
    Stabalize Twisting=true
   )
   &ScaleRibbonWidth(Ribbon Width Scale=&FloatFromCurve(
     FloatCurve=DataInterface() {propertyValues {Curve {keys [
         {interpMode=RCIM_Cubic, tangentMode=RCTM_User, time=0, value=0.5}
         {interpMode=RCIM_Cubic, tangentMode=RCTM_User, time=0.3, value=0.8}
         {arriveTangent=-0.4286, interpMode=RCIM_Cubic, leaveTangent=-0.4286, time=0.7, value=0.8}
         {tangentMode=RCTM_User, time=1, value=0.5}
        ]}}}
     CurveIndex=$Particles.Normalized Spawn Group
    ))
   /SetVariables(Particles.Ribbon02 Color:LinearColor=&MakeLinearColorFromVectorAndFloat(
     "Vector (RGB)"=&MakeVectorFromLinearColorRGB(Color=<rgba 1.5 1.5 1.5 1>)
     "Float (Alpha)"=&Multiply_Float(A=1.5, B=$System.Frontend Alpha NoMusic)
    )) {isSetParameters=true, module=SetVariables_#11}
   /SetVariables(Particles.Position=&AddVectorToPosition(
     Position=$Particles.Position
     Vector=&Multiply_VectorByFloat(
      Vector=&NormalizeVector(
       Vector To Normalize=&SubtractVectorFromPosition(
        Position=&ReturnCameraPosition(Coordinate Space:ENiagaraCoordinateSpace=::Simulation, Camera Query:NiagaraDataInterfaceCamera=DataInterface())
        Vector=Linked() {linkedVariable {name=Particles.Position, type=NiagaraPosition}}
       )
       Safe Normalize=true
       Safe Fallback Threshold=0
       Safe Fallback Vector=<0 0 0.001>
      )
      Float=0
     )
    )) {enabled=false, isSetParameters=true, module=SetVariables_#12}
   &DynamicMaterialParameters(
    Index 0 Param 1=$System.WPO
    Index 0 Param 2=&Lerp_Float(A=0.7, B=1.1, Alpha=$System.Music_Reactive_preview)
    Module.Index0Param0Write=true
    Module.Index0Param1Write=true
    Module.Index0Param2Write=true
    Module.Index0Param3Write=true
   )
   &ScaleColor(
    Scale Alpha=$System.Frontend Alpha NoMusic
    Scale RGB=&Multiply_VectorByFloat(
     Vector=&Multiply_VectorByFloat(
      Vector=<1 1 1>
      Float=&FN_PickFloatDependOnPlatform(TargetPlatformActive=1, TargetPlatformNoActive=0.75)
     )
     Float=$System.TOD
    )
    ScaleA=true
    ScaleRGB=true
   )
  ]
  rendererClasses [&NiagaraRibbonRendererProperties]
  renderers [
   {rendererClass=&NiagaraRibbonRendererProperties}
   {rendererClass=&NiagaraRibbonRendererProperties, rendererIndex=1}
  ]
 }
 Bone_Flow_R_Hand_Gameplay {
  emitterUpdateScript [
   &EmitterState(
    Inactive Response=::NewEnumerator1
    Life Cycle Mode=::NewEnumerator1
    Loop Behavior=::NewEnumerator1
    Loop Duration Mode=::NewEnumerator1
    UseLoopDelay=false
   )
   &SpawnBurst_Instantaneous(
    Spawn Count=&MakeCustomIntFromBool(Bool=$User.bIsFrontEndPreview, True INT=0, False INT=10)
    Use Spawn Probability=false
    UseLoopCountLimit=false
   )
  ]
  particleSpawnScript [
   &FN_SetSpawnGroup(Inclusivity=true, Normalize Spawn Group=true, Particles Per Group=10, Pattern=::NewEnumerator0, Set Ribbon ID and Ribbon Link Order=true)
   &InitializeParticle(
    Color=<rgba 2.2 2.2 2.2 1>
    Ribbon Width=10
    Ribbon Width Mode=::NewEnumerator1
    Sprite Size Mode=::NewEnumerator3
    Uniform Sprite Size=1
    UsePositionOffset=false
    Write Material Random=true
   )
   &CameraOffset(Camera Offset Amount=30)
  ]
  particleUpdateScript [
   /SetVariables(Particles.Ribbon width02:NiagaraFloat=&Multiply_Float(
     A=10
     B=&FloatFromCurve(
      FloatCurve=DataInterface() {propertyValues {Curve {keys [
          {interpMode=RCIM_Cubic, tangentMode=RCTM_User, time=0, value=0.3}
          {interpMode=RCIM_Cubic, tangentMode=RCTM_User, time=0.3, value=0.8}
          {arriveTangent=-0.7143, interpMode=RCIM_Cubic, leaveTangent=-0.7143, time=0.7, value=0.8}
          {tangentMode=RCTM_User, time=1, value=0.3}
         ]}}}
      CurveIndex=$Particles.Normalized Spawn Group
      Scale Curve=&MakeCustomFloatFromBool(Bool=$User.bisHandHidden, True Float=0, False Float=1)
     )
    )) {isSetParameters=true, module=SetVariables_#10}
   &ParticleState(Kill Particles When Lifetime Has Elapsed=false, Let Infinitely Lived Particles Die When Emitter Deactivates=true)
   &FN_BoneFlow(
    Angle:NiagaraFloat=&FN_FrontEndFloatModifier(Gameplay=&Multiply_Float(A=$Particles.Normalized Spawn Group, B=0.3), Front End=0.2)
    BoneID:NiagaraInt32=&ModuloInt(A=$Particles.SpawnGroup, B=1)
    Enable Position Offsets=true
    Height:NiagaraFloat=&Multiply_Float(A=$Particles.Normalized Spawn Group, B=&FN_FrontEndFloatModifier(Gameplay=0.7, Front End=0.6))
    Offset Mid:Vector3f=<0 0 5>
    Offset Start:Vector3f=<-10 0 0>
    Radius:NiagaraFloat=&FloatFromCurve(
     FloatCurve=DataInterface() {propertyValues {Curve {keys [
         {arriveTangent=1.3929, interpMode=RCIM_Cubic, leaveTangent=1.3929, tangentMode=RCTM_User, time=0, value=0.3}
         {interpMode=RCIM_Cubic, tangentMode=RCTM_User, time=0.5, value=1}
         {arriveTangent=-2.2123, tangentMode=RCTM_User, time=1, value=0.2}
        ]}}}
     CurveIndex=$Particles.Normalized Spawn Group
     Scale Curve=10
    )
    SkeletalMesh:NiagaraDataInterfaceSkeletalMesh=$System.Skeletal Mesh RHand
   )
   &ScaleRibbonWidth(Ribbon Width Scale=&FloatFromCurve(
     FloatCurve=DataInterface() {propertyValues {Curve {keys [
         {interpMode=RCIM_Cubic, tangentMode=RCTM_User, time=0, value=0.5}
         {interpMode=RCIM_Cubic, tangentMode=RCTM_User, time=0.3, value=0.8}
         {arriveTangent=-0.4286, interpMode=RCIM_Cubic, leaveTangent=-0.4286, time=0.7, value=0.8}
         {tangentMode=RCTM_User, time=1, value=0.5}
        ]}}}
     CurveIndex=$Particles.Normalized Spawn Group
     Scale Curve=&MakeCustomFloatFromBool(Bool=$User.bisHandHidden, True Float=0, False Float=1)
    ))
   /SetVariables(Particles.Ribbon02 Color:LinearColor=&MakeLinearColorFromVectorAndFloat(
     "Vector (RGB)"=&MakeVectorFromLinearColorRGB(Color=<rgba 1.5 1.5 1.5 1>)
     "Float (Alpha)"=&Multiply_Float(
      A=1.5
      B=&Multiply_Float(
       A=&Lerp_Float(A=0, B=1, Alpha=&ClampFloat(Float=$System.Music_Reactive_preview))
       B=&Multiply_Float(
        A=&Multiply_Float(
         A=$User.Opacity
         B=&FloatFromCurve(
          FloatCurve=DataInterface() {propertyValues {Curve {keys [{tangentMode=RCTM_User, time=0.2, value=0}, {tangentMode=RCTM_User, time=0.5, value=1}]}}}
          CurveIndex=$Particles.Normalized Spawn Group
         )
        )
        B=&FN_FrontEndFloatModifier(
         Gameplay=&FN_LOD_Distance_Remap(
          Start Distance=200
          End Distance=300
          Minimum Output=&Lerp_Float(A=0.1, B=1, Alpha=$System.LOD_Fade_preview)
          Maximum Output=1
          Clamp Output=true
         )
         Front End=1
        )
       )
      )
     )
    )) {isSetParameters=true, module=SetVariables_#11}
   /SetVariables(Particles.Position=&AddVectorToPosition(
     Position=$Particles.Position
     Vector=&Multiply_VectorByFloat(
      Vector=&NormalizeVector(
       Vector To Normalize=&SubtractVectorFromPosition(
        Position=&ReturnCameraPosition(Coordinate Space:ENiagaraCoordinateSpace=::Simulation, Camera Query:NiagaraDataInterfaceCamera=DataInterface())
        Vector=Linked() {linkedVariable {name=Particles.Position, type=NiagaraPosition}}
       )
       Safe Normalize=true
       Safe Fallback Threshold=0
       Safe Fallback Vector=<0 0 0.001>
      )
      Float=0
     )
    )) {enabled=false, isSetParameters=true, module=SetVariables_#12}
   &DynamicMaterialParameters(
    Index 0 Param 1=&Multiply_Float(A=$System.WPO, B=0.8)
    Index 0 Param 2=&Lerp_Float(A=0.7, B=1.1, Alpha=$System.Music_Reactive_preview)
    Module.Index0Param0Write=true
    Module.Index0Param1Write=true
    Module.Index0Param2Write=true
    Module.Index0Param3Write=true
   )
   &ScaleColor(
    Scale Alpha=&Multiply_Float(
     A=&Lerp_Float(A=0, B=1, Alpha=$System.InterpolateOverTime.InterpolateOverTime_Float.Moving Average)
     B=&Multiply_Float(
      A=&Multiply_Float(
       A=$User.Opacity
       B=&FloatFromCurve(
        FloatCurve=DataInterface() {propertyValues {Curve {keys [{tangentMode=RCTM_User, time=0.2, value=1}, {tangentMode=RCTM_User, time=0.5, value=1}]}}}
        CurveIndex=$Particles.Normalized Spawn Group
       )
      )
      B=&FN_FrontEndFloatModifier(
       Gameplay=&FN_LOD_Distance_Remap(
        Start Distance=200
        End Distance=300
        Minimum Output=&Lerp_Float(A=0.05, B=1, Alpha=$System.LOD_Fade_preview)
        Maximum Output=1
        Clamp Output=true
       )
       Front End=1
      )
     )
    )
    Scale RGB=&Multiply_VectorByFloat(
     Vector=&Multiply_VectorByFloat(
      Vector=<1 1 1>
      Float=&FN_PickFloatDependOnPlatform(TargetPlatformActive=1, TargetPlatformNoActive=0.75)
     )
     Float=$System.TOD
    )
    ScaleA=true
    ScaleRGB=true
   )
  ]
  rendererClasses [&NiagaraRibbonRendererProperties]
  renderers [
   {rendererClass=&NiagaraRibbonRendererProperties}
   {rendererClass=&NiagaraRibbonRendererProperties, rendererIndex=1}
  ]
 }
 Bone_Flow_R_Upperarm {
  emitterUpdateScript [
   &EmitterState(
    Inactive Response=::NewEnumerator1
    Life Cycle Mode=::NewEnumerator1
    Loop Behavior=::NewEnumerator1
    Loop Duration Mode=::NewEnumerator1
    UseLoopDelay=false
   )
   &SpawnBurst_Instantaneous(
    Spawn Count=&MakeCustomIntFromBool(Bool=$User.bIsFrontEndPreview, True INT=0, False INT=10)
    Use Spawn Probability=false
    UseLoopCountLimit=false
   )
  ]
  particleSpawnScript [
   &FN_SetSpawnGroup(Inclusivity=true, Normalize Spawn Group=true, Particles Per Group=10, Pattern=::NewEnumerator0, Set Ribbon ID and Ribbon Link Order=true)
   &InitializeParticle(
    Color=<rgba 2.2 2.2 2.2 1>
    Ribbon Width=10
    Ribbon Width Mode=::NewEnumerator1
    Sprite Size Mode=::NewEnumerator3
    Uniform Sprite Size=1
    UsePositionOffset=false
    Write Material Random=true
   )
   &CameraOffset(Camera Offset Amount=30)
  ]
  particleUpdateScript [
   /SetVariables(Particles.Ribbon width02:NiagaraFloat=&Multiply_Float(
     A=15
     B=&FloatFromCurve(
      FloatCurve=DataInterface() {propertyValues {Curve {keys [
          {interpMode=RCIM_Cubic, tangentMode=RCTM_User, time=0, value=0}
          {arriveTangent=0.3, interpMode=RCIM_Cubic, leaveTangent=0.3, time=0.7, value=1}
          {tangentMode=RCTM_User, time=1, value=0.3}
         ]}}}
      CurveIndex=$Particles.Normalized Spawn Group
     )
    )) {isSetParameters=true, module=SetVariables_#13}
   &ParticleState(Kill Particles When Lifetime Has Elapsed=false, Let Infinitely Lived Particles Die When Emitter Deactivates=true)
   &FN_BoneFlow(
    Angle:NiagaraFloat=&Multiply_Float(A=$Particles.Normalized Spawn Group, B=0.5)
    BoneID:NiagaraInt32=&ModuloInt(A=$Particles.SpawnGroup, B=1)
    Height:NiagaraFloat=&Multiply_Float(A=$Particles.Normalized Spawn Group, B=1)
    Radius:NiagaraFloat=&FloatFromCurve(
     FloatCurve=DataInterface() {propertyValues {Curve {keys [
         {interpMode=RCIM_Cubic, tangentMode=RCTM_User, time=0, value=0.4}
         {interpMode=RCIM_Cubic, tangentMode=RCTM_User, time=0.5, value=1}
         {tangentMode=RCTM_User, time=1, value=0.4}
        ]}}}
     CurveIndex=$Particles.Normalized Spawn Group
     Scale Curve=10
    )
    SkeletalMesh:NiagaraDataInterfaceSkeletalMesh=$System.Skeletal Mesh R
   )
   &ScaleRibbonWidth(Ribbon Width Scale=&FloatFromCurve(
     FloatCurve=DataInterface() {propertyValues {Curve {keys [
         {interpMode=RCIM_Cubic, tangentMode=RCTM_User, time=0, value=0.5}
         {interpMode=RCIM_Cubic, tangentMode=RCTM_User, time=0.3, value=1}
         {arriveTangent=-0.7143, interpMode=RCIM_Cubic, leaveTangent=-0.7143, time=0.7, value=1}
         {tangentMode=RCTM_User, time=1, value=0.5}
        ]}}}
     CurveIndex=$Particles.Normalized Spawn Group
    ))
   /SetVariables(Particles.Ribbon02 Color:LinearColor=&MakeLinearColorFromVectorAndFloat(
     "Vector (RGB)"=&MakeVectorFromLinearColorRGB(Color=<rgba 1.5 1.5 1.5 1>)
     "Float (Alpha)"=&Multiply_Float(
      A=1.5
      B=&Multiply_Float(
       A=&Lerp_Float(A=0, B=1, Alpha=&ClampFloat(Float=$System.Music_Reactive_preview))
       B=&Multiply_Float(
        A=$User.Opacity
        B=&FN_FrontEndFloatModifier(
         Gameplay=&FN_LOD_Distance_Remap(
          Start Distance=200
          End Distance=300
          Minimum Output=&Lerp_Float(A=0, B=1, Alpha=$System.LOD_Fade_preview)
          Maximum Output=1
          Clamp Output=true
         )
         Front End=1
        )
       )
      )
     )
    )) {isSetParameters=true, module=SetVariables_#14}
   /SetVariables(Particles.Position=&AddVectorToPosition(
     Position=$Particles.Position
     Vector=&Multiply_VectorByFloat(
      Vector=&NormalizeVector(
       Vector To Normalize=&SubtractVectorFromPosition(
        Position=&ReturnCameraPosition(Coordinate Space:ENiagaraCoordinateSpace=::Simulation, Camera Query:NiagaraDataInterfaceCamera=DataInterface())
        Vector=Linked() {linkedVariable {name=Particles.Position, type=NiagaraPosition}}
       )
       Safe Normalize=true
       Safe Fallback Threshold=0
       Safe Fallback Vector=<0 0 0.001>
      )
      Float=1
     )
    )) {isSetParameters=true, module=SetVariables_#15}
   &DynamicMaterialParameters(
    Index 0 Param 1=$System.WPO
    Index 0 Param 2=&Lerp_Float(A=0.7, B=1.1, Alpha=$System.Music_Reactive_preview)
    Module.Index0Param0Write=true
    Module.Index0Param1Write=true
    Module.Index0Param2Write=true
    Module.Index0Param3Write=true
   )
   &ScaleColor(
    Scale Alpha=&Multiply_Float(
     A=&Multiply_Float(
      A=&Lerp_Float(A=0, B=1, Alpha=$System.InterpolateOverTime.InterpolateOverTime_Float.Moving Average)
      B=$User.Opacity
     )
     B=&FN_FrontEndFloatModifier(
      Gameplay=&FN_LOD_Distance_Remap(
       Start Distance=200
       End Distance=300
       Minimum Output=&Lerp_Float(A=0, B=1, Alpha=$System.LOD_Fade_preview)
       Maximum Output=1
       Clamp Output=true
      )
      Front End=1
     )
    )
    Scale RGB=&Multiply_VectorByFloat(
     Vector=&Multiply_VectorByFloat(
      Vector=<1 1 1>
      Float=&FN_PickFloatDependOnPlatform(TargetPlatformActive=1, TargetPlatformNoActive=0.75)
     )
     Float=$System.TOD
    )
    ScaleA=true
    ScaleRGB=true
   )
  ]
  rendererClasses [&NiagaraRibbonRendererProperties]
  renderers [
   {rendererClass=&NiagaraRibbonRendererProperties}
   {rendererClass=&NiagaraRibbonRendererProperties, rendererIndex=1}
  ]
 }
 Bone_Flow_R_Upperarm001 {
  emitterUpdateScript [
   &EmitterState(
    Inactive Response=::NewEnumerator1
    Life Cycle Mode=::NewEnumerator1
    Loop Behavior=::NewEnumerator1
    Loop Duration Mode=::NewEnumerator1
    UseLoopDelay=false
   )
   &SpawnBurst_Instantaneous(
    Spawn Count=&MakeCustomIntFromBool(Bool=$User.bIsFrontEndPreview, True INT=10, False INT=0)
    Use Spawn Probability=false
    UseLoopCountLimit=false
   )
  ]
  particleSpawnScript [
   &FN_SetSpawnGroup(Inclusivity=true, Normalize Spawn Group=true, Particles Per Group=10, Pattern=::NewEnumerator0, Set Ribbon ID and Ribbon Link Order=true)
   &InitializeParticle(
    Color=<rgba 2.2 2.2 2.2 1>
    Ribbon Width=10
    Ribbon Width Mode=::NewEnumerator1
    Sprite Size Mode=::NewEnumerator3
    Uniform Sprite Size=1
    UsePositionOffset=false
    Write Material Random=true
   )
   &CameraOffset(Camera Offset Amount=30)
  ]
  particleUpdateScript [
   /SetVariables(Particles.Ribbon width02:NiagaraFloat=&Multiply_Float(
     A=15
     B=&FloatFromCurve(
      FloatCurve=DataInterface() {propertyValues {Curve {keys [
          {interpMode=RCIM_Cubic, tangentMode=RCTM_User, time=0, value=0}
          {arriveTangent=0.3, interpMode=RCIM_Cubic, leaveTangent=0.3, time=0.7, value=1}
          {tangentMode=RCTM_User, time=1, value=0.3}
         ]}}}
      CurveIndex=$Particles.Normalized Spawn Group
     )
    )) {isSetParameters=true, module=SetVariables_#13}
   &ParticleState(Kill Particles When Lifetime Has Elapsed=false, Let Infinitely Lived Particles Die When Emitter Deactivates=true)
   &FN_BoneFlow(
    Angle:NiagaraFloat=&Multiply_Float(A=$Particles.Normalized Spawn Group, B=0.5)
    BoneID:NiagaraInt32=&ModuloInt(A=$Particles.SpawnGroup, B=1)
    Height:NiagaraFloat=&Multiply_Float(A=$Particles.Normalized Spawn Group, B=1)
    Radius:NiagaraFloat=&FloatFromCurve(
     FloatCurve=DataInterface() {propertyValues {Curve {keys [
         {interpMode=RCIM_Cubic, tangentMode=RCTM_User, time=0, value=0.4}
         {interpMode=RCIM_Cubic, tangentMode=RCTM_User, time=0.5, value=1}
         {tangentMode=RCTM_User, time=1, value=0.4}
        ]}}}
     CurveIndex=$Particles.Normalized Spawn Group
     Scale Curve=10
    )
    SkeletalMesh:NiagaraDataInterfaceSkeletalMesh=$System.Skeletal Mesh R
   )
   &ScaleRibbonWidth(Ribbon Width Scale=&FloatFromCurve(
     FloatCurve=DataInterface() {propertyValues {Curve {keys [
         {interpMode=RCIM_Cubic, tangentMode=RCTM_User, time=0, value=0.5}
         {interpMode=RCIM_Cubic, tangentMode=RCTM_User, time=0.3, value=1}
         {arriveTangent=-0.7143, interpMode=RCIM_Cubic, leaveTangent=-0.7143, time=0.7, value=1}
         {tangentMode=RCTM_User, time=1, value=0.5}
        ]}}}
     CurveIndex=$Particles.Normalized Spawn Group
    ))
   /SetVariables(Particles.Ribbon02 Color:LinearColor=&MakeLinearColorFromVectorAndFloat(
     "Vector (RGB)"=&MakeVectorFromLinearColorRGB(Color=<rgba 1.5 1.5 1.5 1>)
     "Float (Alpha)"=&Multiply_Float(A=1.5, B=$System.Frontend Alpha NoMusic)
    )) {isSetParameters=true, module=SetVariables_#14}
   /SetVariables(Particles.Position=&AddVectorToPosition(
     Position=$Particles.Position
     Vector=&Multiply_VectorByFloat(
      Vector=&NormalizeVector(
       Vector To Normalize=&SubtractVectorFromPosition(
        Position=&ReturnCameraPosition(Coordinate Space:ENiagaraCoordinateSpace=::Simulation, Camera Query:NiagaraDataInterfaceCamera=DataInterface())
        Vector=Linked() {linkedVariable {name=Particles.Position, type=NiagaraPosition}}
       )
       Safe Normalize=true
       Safe Fallback Threshold=0
       Safe Fallback Vector=<0 0 0.001>
      )
      Float=1
     )
    )) {isSetParameters=true, module=SetVariables_#15}
   &DynamicMaterialParameters(
    Index 0 Param 1=$System.WPO
    Index 0 Param 2=&Lerp_Float(A=0.7, B=1.1, Alpha=$System.Music_Reactive_preview)
    Module.Index0Param0Write=true
    Module.Index0Param1Write=true
    Module.Index0Param2Write=true
    Module.Index0Param3Write=true
   )
   &ScaleColor(
    Scale Alpha=$System.Frontend Alpha NoMusic
    Scale RGB=&Multiply_VectorByFloat(
     Vector=&Multiply_VectorByFloat(
      Vector=<1 1 1>
      Float=&FN_PickFloatDependOnPlatform(TargetPlatformActive=1, TargetPlatformNoActive=0.75)
     )
     Float=$System.TOD
    )
    ScaleA=true
    ScaleRGB=true
   )
  ]
  rendererClasses [&NiagaraRibbonRendererProperties]
  renderers [
   {rendererClass=&NiagaraRibbonRendererProperties}
   {rendererClass=&NiagaraRibbonRendererProperties, rendererIndex=1}
  ]
 }
 ParticalDecal_Light {
  emitterUpdateScript [
   &EmitterState()
   &SpawnRate(SpawnRate=&Lerp_Float(A=0, B=8, Alpha=$System.Music_Reactive_preview), Use Spawn Probability=false)
  ]
  particleSpawnScript [
   &InitializeParticle(
    Color=&MakeCustomLinearColorFromBool(Bool=&RandomBool(), True Color=<rgba 3.5 0.6736 0 1>, False Color=<rgba 3.5 2.2552 0 1>)
    Lifetime=0.3
    Sprite Size Mode=::NewEnumerator3
    Uniform Sprite Size=70
    UsePositionOffset=false
    Write Material Random=true
   )
   &ShapeLocation(
    MaskSpawnGroup=false
    Override Seed=false
    Shape Origin=&VectorFromFloat(Value=0)
    Sphere Radius=20
    UseOwnerScale=true
   )
   &CameraOffset()
  ]
  particleUpdateScript [
   &ParticleState()
   &SkeletalMeshLocation(
    Bone Sampling Mode=::NewEnumerator4
    MaskSpawnGroup=false
    Orientation Sampling=::NewEnumerator1
    Sampled Position Offset=&AddVectorToPosition(Position=$Particles.ShapeLocation.ShapePosition, Vector=<0 0 0>)
    Skeletal Mesh=DataInterface() {propertyValues {
      DefaultMesh=null
      ExcludeBoneName=root
      OptionalComponentClass=null
      PreviewMesh=&M_Med_Soldier_04_Preview
      SkinningMode=SkinOnTheFly
      SoftSourceActor=null
      SourceMode=Default
      UvSetIndex=0
      WholeMeshLOD=-1
      bExcludeBone=false
      bReadDeformedGeometry=true
      bRequireCurrentFrameData=true
      ComponentTags []
      FilteredBones [spine_05]
      FilteredSockets []
      MeshUserParameter {parameter {name=None, typeDefHandle {registeredTypeIndex=1}, varData []}}
      SamplingRegions []
     }}
   )
   &ScaleColor(
    Scale Alpha=&FloatFromCurve(
     FloatCurve=DataInterface() {propertyValues {Curve {keys [
         {interpMode=RCIM_Cubic, tangentMode=RCTM_User, time=0, value=0}
         {interpMode=RCIM_Cubic, tangentMode=RCTM_User, time=0.5, value=1}
         {tangentMode=RCTM_User, time=1, value=0}
        ]}}}
     Scale Curve=$User.Opacity
    )
    Scale RGB=&Multiply_VectorByFloat(Vector=<1 1 1>, Float=&FN_FrontEndFloatModifier(Gameplay=2, Front End=1.5))
    ScaleA=true
    ScaleRGB=true
   )
   &FN_Light_Attributes(Color=<rgba 10 3.9837 0 1>, Create Color Attribute=true, Create Exponent Attribute=true, Create Radius Attribute=true, Exponent=3, Radius=40)
  ]
  rendererClasses [&NiagaraSpriteRendererProperties, &NiagaraLightRendererProperties]
  renderers [
   {rendererClass=&NiagaraSpriteRendererProperties}
   {rendererClass=&NiagaraLightRendererProperties, rendererIndex=1}
  ]
 }
}
summary {
 systemName=NS_Keychain_Skin
 emitters [
  {emitterName=Bone_Flow_R_Upperarm, rendererClasses [&NiagaraRibbonRendererProperties]}
  {emitterName=Bone_Flow_L_Upperarm_GamePlay, rendererClasses [&NiagaraRibbonRendererProperties]}
  {emitterName=Bone_Flow_R_Hand_Gameplay, rendererClasses [&NiagaraRibbonRendererProperties]}
  {emitterName=Bone_Flow_R_Hand_Frontend, rendererClasses [&NiagaraRibbonRendererProperties]}
  {emitterName=Bone_Flow_L_Hand_Frontend, rendererClasses [&NiagaraRibbonRendererProperties]}
  {emitterName=Bone_Flow_L_UpperArm_Frontend, rendererClasses [&NiagaraRibbonRendererProperties]}
  {emitterName=Bone_Flow_L_UpperArm02_Gameplay, rendererClasses [&NiagaraRibbonRendererProperties]}
  {
   emitterName=ParticalDecal_Light
   rendererClasses [&NiagaraSpriteRendererProperties, &NiagaraLightRendererProperties]
  }
  {emitterName=Bone_Flow_R_Upperarm001, rendererClasses [&NiagaraRibbonRendererProperties]}
 ]
 userVariables [
  User.bIsFrontEnd {type=NiagaraBool}
  User.bIsFrontEndPreview {type=NiagaraBool}
  User.bisHandHidden {type=NiagaraBool}
  User.bisInVehicle {type=NiagaraBool}
  User.Debris Opacity {type=NiagaraFloat, defaultValue=1, description=""}
  User.IsSkydiving {type=NiagaraBool}
  User.Opacity {type=NiagaraFloat, defaultValue=1, description=""}
 ]
}
