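# MF_SineCurveVisualizer is a material function in the Keychain character's materials that outputs a hard-edged band mask for drawing a curve line.
# It returns 1 where X lies within Thickness either side of Y, edges included, and 0 elsewhere.
# Two If comparisons against Y plus and minus Thickness are multiplied to form the band.
# Despite the name, it computes no sine itself, only comparing the supplied values.
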MaterialFunction /BRCosmetics/Characters/Player/Male/Medium/Bodies/M_MED_Keychain/Materials/MF_SineCurveVisualizer
graph {nodes [
  n1 = FunctionInput() {inputName=X, inputType=FunctionInput_Scalar}
  n2 = FunctionInput() {inputName=Y, inputType=FunctionInput_Scalar, sortPriority=1}
  n3 = FunctionInput() {inputName=Thickness, inputType=FunctionInput_Scalar, sortPriority=2}
  n4 = Add(A=@n2, B=@n3)
  n5 = Constant()
  n6 = Constant() {r=1}
  n7 = Subtract(A=@n2, B=@n3)
  FunctionOutput(None=Multiply(A=If(A=@n1, B=@n4, "A > B"=@n5, "A == B"=@n6, "A < B"=@n6), B=If(A=@n1, B=@n7, "A > B"=@n6, "A == B"=@n6, "A < B"=@n5)))
 ]}
