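# Material function for the music keychain body cosmetic that paints a glowing sound-wave display inside a masked area, blacking out base color and replacing emissive there.
# Three mirrored sine layers from MF_SineCurveVisualizer plus a center glow are tinted for emission, alongside a two-color background.
# A custom node and CustomRotator place the shape UVs.
# Time and the EventAudio_Amplitude collection parameter, switchable to zero, provide animation and music-reactive inputs.
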MaterialFunction /BRCosmetics/Characters/Player/Male/Medium/Bodies/M_MED_Keychain/Materials/MF_MusicKeyChain_Body
graph {nodes [
  n1 = GetMaterialAttributes(None=FunctionInput() {inputName=MA, inputType=FunctionInput_MaterialAttributes}) {attributeGetTypes [#1, #2]}
  n2 = Add(A=NamedRerouteUsage(), B=NamedRerouteUsage())
  n3 = TextureSampleParameter2D() {
   group=MusicArea
   parameterName=MusicMask
   samplerType=SAMPLERTYPE_Masks
   texture=/BRCosmetics/Characters/Player/Male/Medium/Bodies/M_MED_Keychain/Textures/T_KeyChainMask
  }
  FunctionOutput(None=SetMaterialAttributes!(
    MaterialAttributes=@n1.MaterialAttributes
    EmissiveColor=LinearInterpolate(A=@n1.EmissiveColor, B=@n2, Alpha=@n3.R)
    BaseColor=LinearInterpolate(A=@n1.BaseColor, Alpha=@n3.R) {constB=0}
   ) {attributeSetTypes [#1, #2]})
  NamedRerouteDeclaration(None=@n3.G) {name=EmissionGradient, nodeColor=<rgba 1 0 0.7734 1>}
  NamedRerouteDeclaration(None=LinearInterpolate(
    A=VectorParameter() {group=MusicArea, parameterName=LowColor, defaultValue=<rgba 0.2031 0 0 1>}.RGB
    B=VectorParameter() {group=MusicArea, parameterName=HighColor, defaultValue=<rgba 1 0 0 1>}.RGB
    Alpha=NamedRerouteUsage()
   )) {name=BackGroundColor, nodeColor=<rgba 1 0.9375 0 1>}
  n4 = NamedRerouteUsage()
  n5 = NamedRerouteDeclaration(None=Subtract(A=ComponentMask(None=@n4) {r=true}) {constB=0.5}) {name=XOffseted, nodeColor=<rgba 0 1 0.3672 1>}
  n6 = NamedRerouteDeclaration(None=Subtract(A=ComponentMask(None=@n4) {g=true}, B=ScalarParameter() {defaultValue=-0.02, group=MusicArea, parameterName=SineOffset})) {name=YOffseted, nodeColor=<rgba 1 0.8203 0 1>}
  n7 = Add(A=@n6, B=NamedRerouteUsage())
  n8 = NamedRerouteUsage()
  n9 = NamedRerouteDeclaration(None=LinearInterpolate(Alpha=@n8) {constA=3, constB=18}) {name=WaveMulti, nodeColor=<rgba 0 0.9766 1 1>}
  n10 = NamedRerouteDeclaration(None=ScalarParameter() {defaultValue=-0.2, group=MusicArea, parameterName=SoundWaveStartSmooth}) {name=SmoothMin, nodeColor=<rgba 0 0.5781 1 1>}
  n11 = LinearInterpolate(Alpha=@n8) {constA=1, constB=2}
  n12 = Frac(None=Multiply(A=@n6, B=NamedRerouteDeclaration(None=@n11) {name=WavePhaseMulti, nodeColor=<rgba 0 1 0.9766 1>}))
  n13 = Multiply(
   A=Sine(None=Multiply(A=@n7, B=@n9) {constB=3})
   B=Multiply(A=SmoothStep(Min=@n10, Value=@n12) {constMax=0.85}, B=Constant() {r=0.1}) {constB=0.2}
  ) {constB=0.1}
  n14 = NamedRerouteDeclaration(None=ScalarParameter() {defaultValue=0.013, group=MusicArea, parameterName=Linethickness}) {name=SineCurveThickness, nodeColor=<rgba 0 1 0.2969 1>}
  n15 = Multiply(A=@n13) {constB=-1}
  n16 = Multiply(
   A=Add(
    A=MaterialFunctionCall!(X=@n5, Y=@n13, Thickness=@n14) {materialFunction=&/MF_SineCurveVisualizer}.Result
    B=MaterialFunctionCall(X=@n5, Y=@n15, Thickness=@n14) {materialFunction=&MF_SineCurveVisualizer}.Result
   )
   B=ScalarParameter() {defaultValue=0.3, group=MusicArea, parameterName=FirstEmissionAlpha}
  )
  n17 = NamedRerouteUsage()
  n18 = NamedRerouteUsage()
  n19 = NamedRerouteUsage()
  n20 = Add(A=@n19, B=NamedRerouteUsage())
  n21 = NamedRerouteUsage()
  n22 = NamedRerouteUsage()
  n23 = SmoothStep(Min=NamedRerouteUsage(), Value=Frac(None=Multiply(A=@n19, B=@n22))) {constMax=0.85}
  n24 = Multiply(A=Multiply(A=@n20, B=NamedRerouteUsage())) {constB=2}
  n25 = Multiply(A=ConstantBiasScale(None=Sine(None=@n24))) {constB=2.5}
  n26 = NamedRerouteUsage()
  n27 = Multiply(
   A=Sine(None=Multiply(A=@n20, B=@n21) {constB=3})
   B=Multiply(A=LinearInterpolate(A=@n23, B=@n25, Alpha=@n26), B=Constant() {r=0.05}) {constB=0.2}
  ) {constB=0.1}
  n28 = NamedRerouteUsage()
  n29 = Add(
   A=MaterialFunctionCall(X=@n18, Y=@n27, Thickness=@n28) {materialFunction=&MF_SineCurveVisualizer}.Result
   B=MaterialFunctionCall(X=@n18, Y=Multiply(A=@n27) {constB=-1}, Thickness=@n28) {materialFunction=&MF_SineCurveVisualizer}.Result
  )
  n30 = ScalarParameter() {defaultValue=0.8, group=MusicArea, parameterName=SecondEmissionAlpha}
  n31 = NamedRerouteUsage()
  n32 = NamedRerouteUsage()
  n33 = Add(A=@n32, B=NamedRerouteUsage())
  n34 = NamedRerouteUsage()
  n35 = NamedRerouteUsage()
  n36 = Frac(None=Multiply(A=@n32, B=NamedRerouteUsage()))
  n37 = Multiply(
   A=Sine(None=Multiply(A=@n33, B=@n34) {constB=3})
   B=Multiply(A=SmoothStep(Min=@n35, Value=@n36) {constMax=0.85, constMin=-0.2}, B=Constant() {r=0.13}) {constB=0.2}
  ) {constB=0.1}
  n38 = NamedRerouteUsage()
  n39 = Add(
   A=MaterialFunctionCall(X=@n31, Y=@n37, Thickness=@n38) {materialFunction=&MF_SineCurveVisualizer}.Result
   B=MaterialFunctionCall(X=@n31, Y=Multiply(A=@n37) {constB=-1}, Thickness=@n38) {materialFunction=&MF_SineCurveVisualizer}.Result
  )
  n40 = ScalarParameter() {defaultValue=0.1, group=MusicArea, parameterName=ThirdEmissionAlpha}
  n41 = Add(
   A=Add(A=LinearInterpolate(A=@n16, Alpha=@n17) {constB=0}, B=Multiply(A=@n29, B=@n30))
   B=LinearInterpolate(A=Multiply(A=@n39, B=@n40), Alpha=NamedRerouteUsage()) {constB=0}
  )
  n42 = OneMinus(None=Abs(None=NamedRerouteUsage()))
  n43 = Multiply(A=SmoothStep(Value=Power(Base=@n42) {constExponent=10}) {constMax=0.6}) {constB=0.2}
  n44 = NamedRerouteUsage()
  n45 = Multiply(
   A=Multiply(
    A=Add(A=@n41, B=@n43)
    B=MaterialFunctionCall!(UV_Coord=@n44) {materialFunction=&/MF_UV01_Mask}.Mask
   )
   B=VectorParameter() {parameterName=SineEmission, defaultValue=<rgba 2 0.5638 0 1>}.RGB
  ) {constB=10}
  n46 = ScalarParameter() {defaultValue=2, parameterName=SineEmissionIntensity}
  n47 = NamedRerouteUsage()
  NamedRerouteDeclaration(None=Multiply(A=Multiply(A=@n45, B=@n46), B=LinearInterpolate(Alpha=@n47) {constA=1, constB=0.3})) {name=SineEM, nodeColor=<rgba 0 0.625 1 1>}
  MaterialExpressionTextureCoordinate_0 = TextureCoordinate()
  n48 = Constant2Vector() {g=0.5, r=0.5}
  MaterialExpressionAdd_3 = Add(
   A=AppendVector(
    A=ScalarParameter() {group=MusicArea, parameterName=Shape_UVOffsetX01, sortPriority=62}
    B=ScalarParameter() {group=MusicArea, parameterName=Shape_UVOffsetY01, sortPriority=63}
   )
   B=@n48
  )
  MaterialExpressionScalarParameter_47 = ScalarParameter() {defaultValue=1, group=MusicArea, parameterName=Shape_UVScale01, sortPriority=65}
  n49 = Custom!(OriginalUV=@MaterialExpressionTextureCoordinate_0, UVPoint=@MaterialExpressionAdd_3, UVScale=@MaterialExpressionScalarParameter_47) {
   code="float2 DistUV =OriginalUV-UVPoint;\r\nDistUV = DistUV/(UVScale/2) + float2(0.5,0.5);\r\n\r\nreturn DistUV;"
   description=CopyToPointUV
   outputType=CMOT_Float2
   inputs [
    {
     inputName=OriginalUV
     input {expression=&self:MaterialExpressionTextureCoordinate_0}
    }
    {
     inputName=UVPoint
     input {expression=&self:MaterialExpressionAdd_3}
    }
    {
     inputName=UVScale
     input {expression=&self:MaterialExpressionScalarParameter_47}
    }
   ]
  }
  n50 = MaterialFunctionCall(UV_Coord=@n49) {materialFunction=&MF_UV01_Mask}
  NamedRerouteDeclaration(None=MaterialFunctionCall(
    UVs=Multiply(A=@n49, B=@n50.Mask)
    Rotation Center=@n48
    "Rotation Angle (0-1)"=ScalarParameter() {defaultValue=0.04, group=MusicArea, parameterName=Shape_UVRotation01, sortPriority=64}
   ) {materialFunction=&CustomRotator}.Rotated Values) {name=ShapeUV01, nodeColor=<rgba 0 0.7656 1 1>}
  n51 = NamedRerouteUsage()
  NamedRerouteDeclaration(None=Multiply(A=Time(), B=LinearInterpolate(Alpha=@n51) {constB=0.02})) {name=AnimationSpeed, nodeColor=<rgba 0.6172 1 0 1>}
  n52 = CollectionParameter() {collection=&FortniteMaterialParameters, parameterName=EventAudio_Amplitude}
  NamedRerouteDeclaration(None=StaticSwitchParameter(True=Constant(), False=@n52) {group=MusicArea, parameterName=MusiceReact_0}) {name=MusicAmplitude, nodeColor=<rgba 0 1 0.2266 1>}
 ]}
