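# Physics asset for the Keychain medium male player body, giving its dangling belt bones a physical chain beside the right hip.
# Four capsules (dyn_belt_01 to 04) link in series from a small master box constrained to the pelvis, with pelvis and right-thigh capsules as the only other body shapes.
# Chain joints have tight twist limits and looser swing below the stiffer first link, while the master-to-pelvis joint leaves swing and twist free.
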PhysicsAsset /BRCosmetics/Characters/Player/Male/Medium/Bodies/M_MED_Keychain/Meshes/M_MED_KeyChain_Physics
bodies {
 dyn_belt_01 [{shapeType=Capsule, length=2.5, radius=1.2, shapeName=dyn_belt_01_capsule, center=<2.2 0 0>, rotation=<rot 90 0 0>}]
 dyn_belt_02 [{shapeType=Capsule, length=4.5, radius=2, shapeName=dyn_belt_02_capsule, center=<3.0583 0 0>, rotation=<rot 90 0 0>}]
 dyn_belt_03 [{shapeType=Capsule, length=5, radius=2, shapeName=dyn_belt_03_capsule, center=<3.1026 0 0>, rotation=<rot 90 0 0>}]
 dyn_belt_04 [{shapeType=Capsule, length=7, radius=2, shapeName=dyn_belt_04_capsule, center=<3.6158 0 0>, rotation=<rot 90 0 0>}]
 dyn_belt_MSTR [{shapeType=Box, extentX=1.01, extentY=1.01, extentZ=1.01, shapeName=dyn_belt_MSTR_box, rotation=<rot 0 -90 0>}]
 pelvis [{shapeType=Capsule, length=9.1092, radius=11.2627, center=<0 -2.2479 0>}]
 thigh_r [{shapeType=Capsule, length=22.9511, radius=8.4696, center=<12.4687 -0.4808 0.3407>, rotation=<rot 85.1761 -5.1203 4.2584>}]
}
constraints [
 {
  bone1Name=dyn_belt_MSTR
  bone2Name=pelvis
  swing1LimitDegrees=45
  swing1Motion=Free
  swing2LimitDegrees=45
  swing2Motion=Free
  twistLimitDegrees=45
  twistMotion=Free
 }
 {bone1Name=thigh_r, bone2Name=pelvis, swing1LimitDegrees=45, swing2LimitDegrees=45, twistLimitDegrees=45}
 {bone1Name=dyn_belt_02, bone2Name=dyn_belt_01, swing1LimitDegrees=35, swing2LimitDegrees=35, twistLimitDegrees=3}
 {bone1Name=dyn_belt_01, bone2Name=dyn_belt_MSTR, swing1LimitDegrees=10, swing2LimitDegrees=10, twistLimitDegrees=1}
 {bone1Name=dyn_belt_03, bone2Name=dyn_belt_02, swing1LimitDegrees=35, swing2LimitDegrees=35, twistLimitDegrees=3}
 {bone1Name=dyn_belt_04, bone2Name=dyn_belt_03, swing1LimitDegrees=35, swing2LimitDegrees=35, twistLimitDegrees=3}
]
