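# Physics asset for the left and right dynamic scapula bones on the Keychain medium male player cosmetic body.
# Each side has a capsule on the scapula bone and a tiny sphere on its base bone, alongside one large sphere on spine_04.
# Constraints chain each scapula to its base with tight, asymmetric angular limits, and each base to spine_04 with free swing and twist.
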PhysicsAsset /BRCosmetics/Characters/Player/Male/Medium/Bodies/M_MED_Keychain/Meshes/M_MED_KeyChain_Scapula_Physics
bodies {
 dyn_scapula_l [{shapeType=Capsule, length=11.1481, radius=2.8519, shapeName=dyn_scapula_l_capsule, center=<12 4 0>, rotation=<rot 0 0 -8.5>}]
 dyn_scapula_l_base [{shapeType=Sphere, radius=0.505, shapeName=dyn_scapula_l_base_sphere}]
 dyn_scapula_r [{shapeType=Capsule, length=11.2938, radius=2.7062, shapeName=dyn_scapula_r_capsule, center=<12 -4 0>, rotation=<rot 0 0 8.5>}]
 dyn_scapula_r_base [{shapeType=Sphere, radius=0.505, shapeName=dyn_scapula_r_base_sphere}]
 spine_04 [{shapeType=Sphere, radius=14.4436, shapeName=spine_04_sphere, center=<5.0815 3.6534 -0.0673>}]
}
constraints [
 {
  bone1Name=dyn_scapula_l_base
  bone2Name=spine_04
  swing1LimitDegrees=45
  swing1Motion=Free
  swing2LimitDegrees=45
  swing2Motion=Free
  twistLimitDegrees=45
  twistMotion=Free
 }
 {
  bone1Name=dyn_scapula_r_base
  bone2Name=spine_04
  swing1LimitDegrees=45
  swing1Motion=Free
  swing2LimitDegrees=45
  swing2Motion=Free
  twistLimitDegrees=45
  twistMotion=Free
 }
 {bone1Name=dyn_scapula_l, bone2Name=dyn_scapula_l_base, swing1LimitDegrees=15, swing2LimitDegrees=3, twistLimitDegrees=25}
 {bone1Name=dyn_scapula_r, bone2Name=dyn_scapula_r_base, swing1LimitDegrees=15, swing2LimitDegrees=3, twistLimitDegrees=25}
]
