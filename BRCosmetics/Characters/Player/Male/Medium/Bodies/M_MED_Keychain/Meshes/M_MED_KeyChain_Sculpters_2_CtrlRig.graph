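# Control Rig for the Keychain male medium body that drives corrective deform bones as limbs bend.
# Relative joint angles are remapped to clamped weights that blend local bone overrides, with two angle stages for elbows and knees.
# Execution covers elbows, knees and thighs, then debug-draws the hierarchy; the shoulder pectoral/deltoid chain is disconnected from Begin Execution.
# Elbow, knee and pectoral deform bones are missing from the hierarchy.
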ControlRigBlueprint /BRCosmetics/Characters/Player/Male/Medium/Bodies/M_MED_Keychain/Meshes/M_MED_KeyChain_Sculpters_2_CtrlRig
graphs [
 RigVMFunctionLibrary {links [], nodes []}
 RigVMModel {nodes [
   n1 = &RigUnit_GetRelativeBoneTransform(Bone=upperarm_l, Space=spine_05)
   n2 = &RigVMFunction_MathFloatRemap(
    SourceMaximum=50
    SourceMinimum=170
    TargetMinimum=0
    bClamp=true
    Value=/Script/RigVM.RigVMFunction_MathFloatAbs(Value=&RigVMFunction_MathQuaternionToEuler(RotationOrder=ZYX, Value=@n1.Transform.Rotation).Result.Z) {is=struct RigVMUnitNode, template="Absolute::Execute(in Value,out Result)"}.Result
   )
   n3 = &RigVMFunction_MathQuaternionToEuler(
    RotationOrder=ZYX
    Value=/Script/ControlRig.RigUnit_GetInitialBoneTransform(Bone=deform_pectoral_fr_l, Space=LocalSpace) {is=struct RigVMUnitNode}.Transform.Rotation
   )
   /Script/ControlRig.RigUnit_ModifyBoneTransforms(
    BoneToModify=((Bone="deform_pectoral_fr_l",Transform=(Rotation=(X=0,Y=0,Z=0,W=1),Translation=(X=6.262,Y=5.9,Z=4.98),Scale3D=(X=1,Y=1,Z=1))))
    ExecutePin=()
    Mode=OverrideLocal
    WeightMaximum=1
    Weight=@n2.Result
    BoneToModify.0.Transform.Rotation=&RigVMFunction_MathQuaternionFromEuler(Euler=/Script/RigVM.RigVMFunction_MathVectorAdd(B=(X=0,Y=0,Z=0), A=@n3.Result) {is=struct RigVMUnitNode, template="Add::Execute(in A,in B,out Result)"}.Result).Result
   ) {is=struct RigVMUnitNode}
   n4 = &RigVMFunction_MathQuaternionToEuler(RotationOrder=ZYX, Value=&RigUnit_GetInitialBoneTransform(Bone=deform_deltoid_fr_l, Space=LocalSpace).Transform.Rotation)
   > &RigUnit_ModifyBoneTransforms(
    BoneToModify=((Bone="deform_deltoid_fr_l",Transform=(Rotation=(X=0,Y=0,Z=0,W=1),Translation=(X=5.008,Y=3.542,Z=0),Scale3D=(X=1,Y=1,Z=1))))
    Mode=OverrideLocal
    WeightMaximum=1
    WeightMinimum=0
    Weight=@n2.Result
    BoneToModify.0.Transform.Rotation=&RigVMFunction_MathQuaternionFromEuler(Euler=&RigVMFunction_MathVectorAdd(B=(X=0,Y=0,Z=-51), A=@n4.Result) {template="Add::Execute(in A,in B,out Result)"}.Result).Result
   )
   n5 = &RigUnit_GetRelativeBoneTransform(Bone=upperarm_r, Space=spine_05)
   n6 = &RigVMFunction_MathFloatRemap(
    SourceMaximum=123
    SourceMinimum=30
    TargetMinimum=0
    bClamp=true
    Value=&RigVMFunction_MathFloatAbs(Value=&RigVMFunction_MathQuaternionToEuler(RotationOrder=ZYX, Value=@n5.Transform.Rotation).Result.Z) {template="Absolute::Execute(in Value,out Result)"}.Result
   )
   RigUnit_GetInitialBoneTransform = &RigUnit_GetInitialBoneTransform(Bone=deform_pectoral_fr_r, Space=LocalSpace)
   n7 = &RigVMFunction_MathQuaternionToEuler(RotationOrder=ZYX, Value=@RigUnit_GetInitialBoneTransform.Transform.Rotation)
   > &RigUnit_ModifyBoneTransforms(
    BoneToModify=((Bone="deform_pectoral_fr_r",Transform=(Rotation=(X=0,Y=0,Z=0,W=1),Translation=(X=-6.262,Y=-5.9,Z=-4.98),Scale3D=(X=1,Y=1,Z=1))))
    Mode=OverrideLocal
    WeightMaximum=1
    Weight=@n6.Result
    BoneToModify.0.Transform.Rotation=&RigVMFunction_MathQuaternionFromEuler(Euler=&RigVMFunction_MathVectorAdd(B=(X=0,Y=0,Z=0), A=@n7.Result) {template="Add::Execute(in A,in B,out Result)"}.Result).Result
   )
   n8 = &RigVMFunction_MathQuaternionToEuler(RotationOrder=ZYX, Value=&RigUnit_GetInitialBoneTransform(Bone=deform_deltoid_fr_r, Space=LocalSpace).Transform.Rotation)
   > &RigUnit_ModifyBoneTransforms(
    BoneToModify=((Bone="deform_deltoid_fr_r",Transform=(Rotation=(X=0,Y=0,Z=0,W=1),Translation=(X=-5.008,Y=-3.542,Z=0),Scale3D=(X=1,Y=1,Z=1))))
    Mode=OverrideLocal
    WeightMaximum=1
    WeightMinimum=0
    Weight=@n6.Result
    BoneToModify.0.Transform.Rotation=&RigVMFunction_MathQuaternionFromEuler(Euler=&RigVMFunction_MathVectorAdd(B=(X=0,Y=0,Z=-51), A=@n8.Result) {template="Add::Execute(in A,in B,out Result)"}.Result).Result
   )
   RigUnit_BeginExecution = &RigUnit_BeginExecution()
   RigUnit_GetRelativeBoneTransform = &RigUnit_GetRelativeBoneTransform(Bone=lowerarm_l, Space=upperarm_l)
   n9 = &RigVMFunction_MathQuaternionToEuler(RotationOrder=ZYX, Value=@RigUnit_GetRelativeBoneTransform.Transform.Rotation)
   n10 = &RigVMFunction_MathFloatRemap(SourceMaximum=92, SourceMinimum=28, TargetMinimum=0, bClamp=true, Value=@n9.Result.Z)
   n11 = &RigVMFunction_MathQuaternionToEuler(RotationOrder=ZYX, Value=&RigUnit_GetInitialBoneTransform(Bone=deform_elbow_bk_l, Space=LocalSpace).Transform.Rotation)
   &RigUnit_ModifyBoneTransforms(
    BoneToModify=((Bone="deform_elbow_bk_l",Transform=(Rotation=(X=0,Y=0,Z=0,W=1),Translation=(X=-2.858,Y=-2.932,Z=-0.666),Scale3D=(X=1,Y=1,Z=1))))
    Mode=OverrideLocal
    WeightMaximum=1
    Weight=@n10.Result
    BoneToModify.0.Transform.Rotation=&RigVMFunction_MathQuaternionFromEuler(Euler=&RigVMFunction_MathVectorAdd(B=(X=0,Y=0,Z=0), A=@n11.Result) {template="Add::Execute(in A,in B,out Result)"}.Result).Result
    ExecutePin=@RigUnit_BeginExecution.ExecutePin
   )
   n12 = &RigVMFunction_MathQuaternionToEuler(RotationOrder=ZYX, Value=&RigUnit_GetInitialBoneTransform(Bone=deform_elbow_fr_l, Space=LocalSpace).Transform.Rotation)
   > &RigUnit_ModifyBoneTransforms(
    BoneToModify=((Bone="deform_elbow_fr_l",Transform=(Rotation=(X=0,Y=0,Z=0,W=1),Translation=(X=24.635,Y=2.261,Z=-0.637),Scale3D=(X=1,Y=1,Z=1.125))))
    Mode=OverrideLocal
    WeightMaximum=1
    WeightMinimum=0
    Weight=@n10.Result
    BoneToModify.0.Transform.Rotation=&RigVMFunction_MathQuaternionFromEuler(Euler=&RigVMFunction_MathVectorAdd(B=(X=0,Y=0,Z=0), A=@n12.Result) {template="Add::Execute(in A,in B,out Result)"}.Result).Result
   )
   n13 = &RigVMFunction_MathFloatRemap(SourceMaximum=148, SourceMinimum=91, TargetMinimum=0, bClamp=true, Value=@n9.Result.Z)
   > &RigUnit_ModifyBoneTransforms(
    BoneToModify=((Bone="deform_elbow_bk_l",Transform=(Rotation=(X=0,Y=0,Z=0,W=1),Translation=(X=-3.59,Y=-2.934,Z=-0.716),Scale3D=(X=1,Y=1,Z=1))))
    Mode=OverrideLocal
    WeightMaximum=1
    Weight=@n13.Result
    BoneToModify.0.Transform.Rotation=&RigVMFunction_MathQuaternionFromEuler(Euler=&RigVMFunction_MathVectorAdd(B=(X=0,Y=0,Z=-13.6), A=@n11.Result) {template="Add::Execute(in A,in B,out Result)"}.Result).Result
   )
   > &RigUnit_ModifyBoneTransforms(
    BoneToModify=((Bone="deform_elbow_fr_l",Transform=(Rotation=(X=0,Y=0,Z=0,W=1),Translation=(X=15.651,Y=1.255,Z=-0.922),Scale3D=(X=1,Y=1,Z=1.337))))
    Mode=OverrideLocal
    WeightMaximum=1
    WeightMinimum=0
    Weight=@n13.Result
    BoneToModify.0.Transform.Rotation=&RigVMFunction_MathQuaternionFromEuler(Euler=&RigVMFunction_MathVectorAdd(B=(X=0,Y=0,Z=28), A=@n12.Result) {template="Add::Execute(in A,in B,out Result)"}.Result).Result
   )
   n14 = &RigVMFunction_MathQuaternionToEuler(RotationOrder=ZYX, Value=&RigUnit_GetRelativeBoneTransform(Bone=lowerarm_r, Space=upperarm_r).Transform.Rotation)
   n15 = &RigVMFunction_MathFloatRemap(SourceMaximum=92, SourceMinimum=28, TargetMinimum=0, bClamp=true, Value=@n14.Result.Z)
   n16 = &RigVMFunction_MathQuaternionToEuler(RotationOrder=ZYX, Value=&RigUnit_GetInitialBoneTransform(Bone=deform_elbow_bk_r, Space=LocalSpace).Transform.Rotation)
   > &RigUnit_ModifyBoneTransforms(
    BoneToModify=((Bone="deform_elbow_bk_r",Transform=(Rotation=(X=0,Y=0,Z=0,W=1),Translation=(X=2.858,Y=2.932,Z=0.666),Scale3D=(X=1,Y=1,Z=1))))
    Mode=OverrideLocal
    WeightMaximum=1
    Weight=@n15.Result
    BoneToModify.0.Transform.Rotation=&RigVMFunction_MathQuaternionFromEuler(Euler=&RigVMFunction_MathVectorAdd(B=(X=0,Y=0,Z=0), A=@n16.Result) {template="Add::Execute(in A,in B,out Result)"}.Result).Result
   )
   n17 = &RigVMFunction_MathQuaternionToEuler(RotationOrder=ZYX, Value=&RigUnit_GetInitialBoneTransform(Bone=deform_elbow_fr_r, Space=LocalSpace).Transform.Rotation)
   > &RigUnit_ModifyBoneTransforms(
    BoneToModify=((Bone="deform_elbow_fr_r",Transform=(Rotation=(X=0,Y=0,Z=0,W=1),Translation=(X=-24.635,Y=-2.261,Z=0.637),Scale3D=(X=1,Y=1,Z=1.125))))
    Mode=OverrideLocal
    WeightMaximum=1
    WeightMinimum=0
    Weight=@n15.Result
    BoneToModify.0.Transform.Rotation=&RigVMFunction_MathQuaternionFromEuler(Euler=&RigVMFunction_MathVectorAdd(B=(X=0,Y=0,Z=0), A=@n17.Result) {template="Add::Execute(in A,in B,out Result)"}.Result).Result
   )
   n18 = &RigVMFunction_MathFloatRemap(SourceMaximum=148, SourceMinimum=91, TargetMinimum=0, bClamp=true, Value=@n14.Result.Z)
   > &RigUnit_ModifyBoneTransforms(
    BoneToModify=((Bone="deform_elbow_bk_r",Transform=(Rotation=(X=0,Y=0,Z=0,W=1),Translation=(X=3.59,Y=2.934,Z=0.716),Scale3D=(X=1,Y=1,Z=1))))
    Mode=OverrideLocal
    WeightMaximum=1
    Weight=@n18.Result
    BoneToModify.0.Transform.Rotation=&RigVMFunction_MathQuaternionFromEuler(Euler=&RigVMFunction_MathVectorAdd(B=(X=0,Y=0,Z=-13.6), A=@n16.Result) {template="Add::Execute(in A,in B,out Result)"}.Result).Result
   )
   > &RigUnit_ModifyBoneTransforms(
    BoneToModify=((Bone="deform_elbow_fr_r",Transform=(Rotation=(X=0,Y=0,Z=0,W=1),Translation=(X=-15.651,Y=-1.255,Z=0.922),Scale3D=(X=1,Y=1,Z=1.337))))
    Mode=OverrideLocal
    WeightMaximum=1
    WeightMinimum=0
    Weight=@n18.Result
    BoneToModify.0.Transform.Rotation=&RigVMFunction_MathQuaternionFromEuler(Euler=&RigVMFunction_MathVectorAdd(B=(X=0,Y=0,Z=28), A=@n17.Result) {template="Add::Execute(in A,in B,out Result)"}.Result).Result
   )
   n19 = &RigVMFunction_MathQuaternionToEuler(RotationOrder=ZYX, Value=&RigUnit_GetRelativeBoneTransform(Bone=calf_l, Space=thigh_l).Transform.Rotation)
   n20 = &RigVMFunction_MathFloatRemap(SourceMaximum=114, SourceMinimum=4, TargetMinimum=0, bClamp=true, Value=@n19.Result.Z)
   n21 = &RigVMFunction_MathQuaternionToEuler(RotationOrder=ZYX, Value=&RigUnit_GetInitialBoneTransform(Bone=deform_knee_bk_01_l, Space=LocalSpace).Transform.Rotation)
   > &RigUnit_ModifyBoneTransforms(
    BoneToModify=((Bone="deform_knee_bk_01_l",Transform=(Rotation=(X=0,Y=0,Z=0,W=1),Translation=(X=-4.319,Y=-4.798,Z=0),Scale3D=(X=1,Y=1,Z=0.7))))
    Mode=OverrideLocal
    WeightMaximum=1
    Weight=@n20.Result
    BoneToModify.0.Transform.Rotation=&RigVMFunction_MathQuaternionFromEuler(Euler=&RigVMFunction_MathVectorAdd(B=(X=0,Y=0,Z=0), A=@n21.Result) {template="Add::Execute(in A,in B,out Result)"}.Result).Result
   )
   n22 = &RigVMFunction_MathQuaternionToEuler(RotationOrder=ZYX, Value=&RigUnit_GetInitialBoneTransform(Bone=deform_knee_fr_01_l, Space=LocalSpace).Transform.Rotation)
   > &RigUnit_ModifyBoneTransforms(
    BoneToModify=((Bone="deform_knee_fr_01_l",Transform=(Rotation=(X=0,Y=0,Z=0,W=1),Translation=(X=0.258,Y=-2,Z=0.009),Scale3D=(X=1,Y=1,Z=1))))
    Mode=OverrideLocal
    WeightMaximum=1
    WeightMinimum=0
    Weight=@n20.Result
    BoneToModify.0.Transform.Rotation=&RigVMFunction_MathQuaternionFromEuler(Euler=&RigVMFunction_MathVectorAdd(B=(X=0,Y=0,Z=-15), A=@n22.Result) {template="Add::Execute(in A,in B,out Result)"}.Result).Result
   )
   n23 = &RigVMFunction_MathFloatRemap(SourceMaximum=130, SourceMinimum=113, TargetMinimum=0, bClamp=true, Value=@n19.Result.Z)
   > &RigUnit_ModifyBoneTransforms(
    BoneToModify=((Bone="deform_knee_bk_01_l",Transform=(Rotation=(X=0,Y=0,Z=0,W=1),Translation=(X=-10.633,Y=-3.561,Z=0),Scale3D=(X=1,Y=1,Z=1))))
    Mode=OverrideLocal
    WeightMaximum=1
    WeightMinimum=0
    Weight=@n23.Result
    BoneToModify.0.Transform.Rotation=&RigVMFunction_MathQuaternionFromEuler(Euler=&RigVMFunction_MathVectorAdd(B=(X=0,Y=0,Z=0), A=@n21.Result) {template="Add::Execute(in A,in B,out Result)"}.Result).Result
   )
   > &RigUnit_ModifyBoneTransforms(
    BoneToModify=((Bone="deform_knee_fr_01_l",Transform=(Rotation=(X=0,Y=0,Z=0,W=1),Translation=(X=0.258,Y=-3.591,Z=0.009),Scale3D=(X=1,Y=1,Z=1))))
    Mode=OverrideLocal
    WeightMaximum=1
    WeightMinimum=0
    Weight=@n23.Result
    BoneToModify.0.Transform.Rotation=&RigVMFunction_MathQuaternionFromEuler(Euler=&RigVMFunction_MathVectorAdd(B=(X=0,Y=0,Z=-51.2), A=@n22.Result) {template="Add::Execute(in A,in B,out Result)"}.Result).Result
   )
   n24 = &RigVMFunction_MathQuaternionToEuler(RotationOrder=ZYX, Value=&RigUnit_GetRelativeBoneTransform(Bone=calf_r, Space=thigh_r).Transform.Rotation)
   n25 = &RigVMFunction_MathFloatRemap(SourceMaximum=90, SourceMinimum=4, TargetMinimum=0, bClamp=true, Value=@n24.Result.Z)
   n26 = &RigVMFunction_MathQuaternionToEuler(RotationOrder=ZYX, Value=&RigUnit_GetInitialBoneTransform(Bone=deform_knee_bk_01_r, Space=LocalSpace).Transform.Rotation)
   > &RigUnit_ModifyBoneTransforms(
    BoneToModify=((Bone="deform_knee_bk_01_r",Transform=(Rotation=(X=0,Y=0,Z=0,W=1),Translation=(X=4.319,Y=4.798,Z=0),Scale3D=(X=1,Y=2.022,Z=1.95))))
    Mode=OverrideLocal
    WeightMaximum=1
    Weight=@n25.Result
    BoneToModify.0.Transform.Rotation=&RigVMFunction_MathQuaternionFromEuler(Euler=&RigVMFunction_MathVectorAdd(B=(X=0,Y=0,Z=0), A=@n26.Result) {template="Add::Execute(in A,in B,out Result)"}.Result).Result
   )
   n27 = &RigVMFunction_MathQuaternionToEuler(RotationOrder=ZYX, Value=&RigUnit_GetInitialBoneTransform(Bone=deform_knee_fr_01_r, Space=LocalSpace).Transform.Rotation)
   > &RigUnit_ModifyBoneTransforms(
    BoneToModify=((Bone="deform_knee_fr_01_r",Transform=(Rotation=(X=0,Y=0,Z=0,W=1),Translation=(X=-1.701,Y=-1.197,Z=0),Scale3D=(X=1,Y=1,Z=1))))
    Mode=OverrideLocal
    WeightMaximum=1
    WeightMinimum=0
    Weight=@n25.Result
    BoneToModify.0.Transform.Rotation=&RigVMFunction_MathQuaternionFromEuler(Euler=&RigVMFunction_MathVectorAdd(B=(X=0,Y=0,Z=0), A=@n27.Result) {template="Add::Execute(in A,in B,out Result)"}.Result).Result
   )
   n28 = &RigVMFunction_MathFloatRemap(SourceMaximum=130, SourceMinimum=89, TargetMinimum=0, bClamp=true, Value=@n24.Result.Z)
   > &RigUnit_ModifyBoneTransforms(
    BoneToModify=((Bone="deform_knee_bk_01_r",Transform=(Rotation=(X=0,Y=0,Z=0,W=1),Translation=(X=10.633,Y=3.561,Z=0),Scale3D=(X=1,Y=1,Z=1.777))))
    Mode=OverrideLocal
    WeightMaximum=1
    WeightMinimum=0
    Weight=@n28.Result
    BoneToModify.0.Transform.Rotation=&RigVMFunction_MathQuaternionFromEuler(Euler=&RigVMFunction_MathVectorAdd(B=(X=0,Y=0,Z=0), A=@n26.Result) {template="Add::Execute(in A,in B,out Result)"}.Result).Result
   )
   > &RigUnit_ModifyBoneTransforms(
    BoneToModify=((Bone="deform_knee_fr_01_r",Transform=(Rotation=(X=0,Y=0,Z=0,W=1),Translation=(X=-1.791,Y=-1.35,Z=0),Scale3D=(X=1.913,Y=1,Z=1))))
    Mode=OverrideLocal
    WeightMaximum=1
    WeightMinimum=0
    Weight=@n28.Result
    BoneToModify.0.Transform.Rotation=&RigVMFunction_MathQuaternionFromEuler(Euler=&RigVMFunction_MathVectorAdd(B=(X=0,Y=0,Z=0), A=@n27.Result) {template="Add::Execute(in A,in B,out Result)"}.Result).Result
   )
   n29 = &RigUnit_GetRelativeBoneTransform(Bone=thigh_l, Space=pelvis)
   n30 = &RigVMFunction_MathFloatRemap(
    SourceMaximum=100
    SourceMinimum=2
    TargetMinimum=0
    bClamp=true
    Value=&RigVMFunction_MathFloatAbs(Value=&RigVMFunction_MathQuaternionToEuler(RotationOrder=ZYX, Value=@n29.Transform.Rotation).Result.Z) {template="Absolute::Execute(in Value,out Result)"}.Result
   )
   n31 = &RigVMFunction_MathQuaternionToEuler(RotationOrder=ZYX, Value=&RigUnit_GetInitialBoneTransform(Bone=deform_thigh_fr_l, Space=LocalSpace).Transform.Rotation)
   > &RigUnit_ModifyBoneTransforms(
    BoneToModify=((Bone="deform_thigh_fr_l",Transform=(Rotation=(X=0,Y=0,Z=0,W=1),Translation=(X=13.423,Y=8.962,Z=2.253),Scale3D=(X=1,Y=1,Z=1.175))))
    Mode=OverrideLocal
    WeightMaximum=1
    Weight=@n30.Result
    BoneToModify.0.Transform.Rotation=&RigVMFunction_MathQuaternionFromEuler(Euler=&RigVMFunction_MathVectorAdd(B=(X=0,Y=-14.949,Z=0), A=@n31.Result) {template="Add::Execute(in A,in B,out Result)"}.Result).Result
   )
   n32 = &RigVMFunction_MathQuaternionToEuler(RotationOrder=ZYX, Value=&RigUnit_GetInitialBoneTransform(Bone=deform_thigh_bk_l, Space=LocalSpace).Transform.Rotation)
   > &RigUnit_ModifyBoneTransforms(
    BoneToModify=((Bone="deform_thigh_bk_l",Transform=(Rotation=(X=0,Y=0,Z=0,W=1),Translation=(X=29.958,Y=-1.401,Z=0.029),Scale3D=(X=1,Y=1,Z=1.803))))
    Mode=OverrideLocal
    WeightMaximum=1
    WeightMinimum=0
    Weight=@n30.Result
    BoneToModify.0.Transform.Rotation=&RigVMFunction_MathQuaternionFromEuler(Euler=&RigVMFunction_MathVectorAdd(B=(X=0,Y=-37.217,Z=0), A=@n32.Result) {template="Add::Execute(in A,in B,out Result)"}.Result).Result
   )
   n33 = &RigUnit_GetRelativeBoneTransform(Bone=thigh_r, Space=pelvis)
   n34 = &RigVMFunction_MathFloatRemap(
    SourceMaximum=77
    SourceMinimum=177
    TargetMinimum=0
    bClamp=true
    Value=&RigVMFunction_MathFloatAbs(Value=&RigVMFunction_MathQuaternionToEuler(RotationOrder=ZYX, Value=@n33.Transform.Rotation).Result.Z) {template="Absolute::Execute(in Value,out Result)"}.Result
   )
   n35 = &RigVMFunction_MathQuaternionToEuler(RotationOrder=ZYX, Value=&RigUnit_GetInitialBoneTransform(Bone=deform_thigh_fr_r, Space=LocalSpace).Transform.Rotation)
   > &RigUnit_ModifyBoneTransforms(
    BoneToModify=((Bone="deform_thigh_fr_r",Transform=(Rotation=(X=0,Y=0,Z=0,W=1),Translation=(X=-13.423,Y=-8.962,Z=-2.253),Scale3D=(X=1,Y=1,Z=1.175))))
    Mode=OverrideLocal
    WeightMaximum=1
    Weight=@n34.Result
    BoneToModify.0.Transform.Rotation=&RigVMFunction_MathQuaternionFromEuler(Euler=&RigVMFunction_MathVectorAdd(B=(X=0,Y=-14.949,Z=0), A=@n35.Result) {template="Add::Execute(in A,in B,out Result)"}.Result).Result
   )
   n36 = &RigVMFunction_MathQuaternionToEuler(RotationOrder=ZYX, Value=&RigUnit_GetInitialBoneTransform(Bone=deform_thigh_bk_r, Space=LocalSpace).Transform.Rotation)
   > &RigUnit_ModifyBoneTransforms(
    BoneToModify=((Bone="deform_thigh_bk_r",Transform=(Rotation=(X=0,Y=0,Z=0,W=1),Translation=(X=-29.958,Y=1.401,Z=-0.029),Scale3D=(X=1,Y=1,Z=1.803))))
    Mode=OverrideLocal
    WeightMaximum=1
    WeightMinimum=0
    Weight=@n34.Result
    BoneToModify.0.Transform.Rotation=&RigVMFunction_MathQuaternionFromEuler(Euler=&RigVMFunction_MathVectorAdd(B=(X=0,Y=-37.217,Z=0), A=@n36.Result) {template="Add::Execute(in A,in B,out Result)"}.Result).Result
   )
   > /Script/ControlRig.RigUnit_DebugHierarchy(
    Color=(R=1,G=1,B=1,A=1)
    DebugDrawSettings=(DepthPriority=SDPG_Foreground,Lifetime=-1)
    Items=()
    Scale=3
    Thickness=0.1
    WorldOffset=(Rotation=(X=0,Y=0,Z=0,W=1),Translation=(X=0,Y=0,Z=0),Scale3D=(X=1,Y=1,Z=1))
    bEnabled=True
   ) {is=struct RigVMUnitNode}
   RigVMVariableNode() {variable=KneeRot_fr_l_114deg}
   RigVMVariableNode() {variable=KneeTrans_fr_l_114deg}
   RigVMVariableNode() {variable=KneeScale_fr_l_114deg}
   RigVMVariableNode() {variable=KneeScale_fr_l_130deg}
   RigVMVariableNode() {variable=KneeRot_fr_l_130deg}
   RigVMVariableNode() {variable=KneeTrans_fr_l_130deg}
  ]}
]
hierarchy [
 root {loc=<0 0 0>, rot=<0 0 0 1>}
 pelvis {parent=root, loc=<0 -0.7071 93.1191>, rot=<0.0127 0.707 -0.0127 -0.707>}
 spine_01 {parent=pelvis, loc=<6.7494 -0.6388 0>, rot=<0 0 0.034 -0.9994>}
 spine_02 {parent=spine_01, loc=<7.2518 -0.035 0>, rot=<0 0 0.0238 -0.9997>}
 spine_03 {parent=spine_02, loc=<6.0104 0 0>, rot=<0 0 0.0149 -0.9999>}
 spine_04 {parent=spine_03, loc=<8.1024 0 0>, rot=<0 0 -0.0132 -0.9999>}
 spine_05 {parent=spine_04, loc=<10.0146 0 0>, rot=<0 0 -0.0015 -1>}
 attach_cape {parent=spine_05, loc=<9.017 -11.714 0>, rot=<0.0616 -0.7044 -0.0616 -0.7044>}
 neck_01 {parent=spine_05, loc=<11.3138 0.1529 0>, rot=<0 0 -0.2003 -0.9797>}
 neck_02 {parent=neck_01, loc=<4.5627 0 0>, rot=<0 0 0.0504 -0.9987>}
 head {parent=neck_02, loc=<5.3914 0 0>, rot=<0 0 0.115 -0.9934>}
 dyn_neck_01 {parent=neck_01, loc=<0 0 0>, rot=<0 0 0 -1>}
 attach_backpack {parent=spine_05, loc=<-0.123 -13.473 0>, rot=<-0.0093 -0.707 0.0093 -0.707>}
 clavicle_l {parent=spine_05, loc=<4.3575 -0.1466 -2.3223>, rot=<0.0465 -0.7289 -0.009 -0.6829>}
 upperarm_l {parent=clavicle_l, loc=<15.5111 0 0>, rot=<0.0226 -0.3234 -0.0088 -0.9459>}
 lowerarm_l {parent=upperarm_l, loc=<26.6514 0 0>, rot=<0.0391 -0.0099 -0.2453 -0.9686>}
 hand_l {parent=lowerarm_l, loc=<22.3491 0 0>, rot=<-0.7081 -0.0507 0.0975 -0.6975>}
 thumb_01_l {parent=hand_l, loc=<1.3118 -0.3191 -2.459>, rot=<0.6546 -0.4042 0.0664 -0.6354>}
 thumb_02_l {parent=thumb_01_l, loc=<5.9336 -0.0448 0.0401>, rot=<-0.0037 0.0213 0.298 -0.9543>}
 thumb_03_l {parent=thumb_02_l, loc=<2.2223 0.3682 -0.0907>, rot=<0.0118 -0.013 0.0066 -0.9998>}
 index_metacarpal_l {parent=hand_l, loc=<1.6535 0.0183 -2.96>, rot=<0.0158 -0.0124 -0.0173 -0.9996>}
 index_01_l {parent=index_metacarpal_l, loc=<8.3008 0.0357 -0.0524>, rot=<-0.0044 0.0485 0.0924 -0.9945>}
 index_02_l {parent=index_01_l, loc=<3.3587 -0.3004 0.0452>, rot=<-0.002 0.0027 0.3225 -0.9466>}
 index_03_l {parent=index_02_l, loc=<2.5578 -0.123 -0.0252>, rot=<-0.002 0.0067 0.2558 -0.9667>}
 middle_metacarpal_l {parent=hand_l, loc=<2.1856 -0.1725 -0.5984>, rot=<0.006 0.0168 -0.0329 -0.9993>}
 middle_01_l {parent=middle_metacarpal_l, loc=<7.3275 -0.2447 0.0129>, rot=<-0.0208 0.0042 0.0858 -0.9961>}
 middle_02_l {parent=middle_01_l, loc=<4.0123 0.1401 0.1175>, rot=<-0.0218 0.0258 0.3512 -0.9357>}
 middle_03_l {parent=middle_02_l, loc=<2.6077 0.0534 0.1373>, rot=<0.0082 0.0188 0.1708 -0.9851>}
 ring_metacarpal_l {parent=hand_l, loc=<2.2074 -0.3625 1.3638>, rot=<-0.0284 0.0722 -0.0458 -0.9959>}
 ring_01_l {parent=ring_metacarpal_l, loc=<7.0295 -0.6197 -0.1446>, rot=<-0.0489 -0.0252 0.1892 -0.9804>}
 ring_02_l {parent=ring_01_l, loc=<3.6174 0.4647 -0.1027>, rot=<0 0 0.2806 -0.9598>}
 ring_03_l {parent=ring_02_l, loc=<2.4152 0.114 0.0199>, rot=<0 0 0.1189 -0.9929>}
 pinky_metacarpal_l {parent=hand_l, loc=<1.7616 -0.9753 3.3814>, rot=<-0.0397 0.0814 -0.069 -0.9935>}
 pinky_01_l {parent=pinky_metacarpal_l, loc=<7.1709 -0.4831 -0.1094>, rot=<-0.0654 -0.0035 0.2242 -0.9723>}
 pinky_02_l {parent=pinky_01_l, loc=<2.8274 0.0893 -0.1043>, rot=<0.0037 0.0011 0.2854 -0.9584>}
 pinky_03_l {parent=pinky_02_l, loc=<2.2046 0.0307 -0.0078>, rot=<-0.0002 0.0098 0.1445 -0.9895>}
 dyn_hand {parent=hand_l, loc=<0 0 0>, rot=<0.0649 -0.8055 -0.5823 0.0887>}
 dyn_radius {parent=dyn_hand, loc=<0 0 0>, rot=<0 0 0 -1>}
 dyn_ulna {parent=dyn_hand, loc=<0 0 0>, rot=<0 0 0 1>}
 lowerarm_twist_02_l {parent=lowerarm_l, loc=<9.6199 0 0>, rot=<0 0 0 -1>}
 lowerarm_twist_01_l {parent=lowerarm_l, loc=<17.8549 0 0>, rot=<0 0 0 -1>}
 dyn_radius_lookat {parent=lowerarm_l, loc=<0 0 0>, rot=<0 0.1513 0.9885 0>}
 upperarm_twist_01_l {parent=upperarm_l, loc=<0.0041 0 0>, rot=<0 0 0 1>}
 deform_deltoid_fr_l {parent=upperarm_twist_01_l, loc=<2.4511 0 0>, rot=<0 0 0 -1>}
 upperarm_twist_02_l {parent=upperarm_l, loc=<13.7941 0 0>, rot=<0 0 0 1>}
 dyn_upperarm_l {parent=clavicle_l, loc=<15.5111 0 0>, rot=<0.0008 0.0336 -0.0252 -0.9991>}
 dyn_scapula_l_base {parent=dyn_upperarm_l, loc=<0 0 0>, rot=<-0.0994 -0.0222 0.9708 -0.217>}
 dyn_scapula_l {parent=dyn_scapula_l_base, loc=<0 0 0>, rot=<0 0 0 1>}
 clavicle_r {parent=spine_05, loc=<4.3575 -0.1466 2.3223>, rot=<0.7289 0.0465 -0.6829 0.009>}
 upperarm_r {parent=clavicle_r, loc=<-15.5111 0 0>, rot=<0.0226 -0.3234 -0.0088 -0.9459>}
 lowerarm_r {parent=upperarm_r, loc=<-26.6514 0 0>, rot=<0.0391 -0.0099 -0.2453 -0.9686>}
 hand_r {parent=lowerarm_r, loc=<-22.3491 0 0>, rot=<-0.7081 -0.0507 0.0975 -0.6975>}
 thumb_01_r {parent=hand_r, loc=<-1.3118 0.3191 2.459>, rot=<0.6546 -0.4042 0.0664 -0.6354>}
 thumb_02_r {parent=thumb_01_r, loc=<-5.9336 0.0448 -0.0401>, rot=<-0.0037 0.0213 0.298 -0.9543>}
 thumb_03_r {parent=thumb_02_r, loc=<-2.2223 -0.3682 0.0907>, rot=<0.0118 -0.013 0.0066 -0.9998>}
 index_metacarpal_r {parent=hand_r, loc=<-1.6535 -0.0183 2.96>, rot=<0.0158 -0.0124 -0.0173 -0.9996>}
 index_01_r {parent=index_metacarpal_r, loc=<-8.3008 -0.0357 0.0524>, rot=<-0.0044 0.0485 0.0924 -0.9945>}
 index_02_r {parent=index_01_r, loc=<-3.3587 0.3004 -0.0452>, rot=<-0.002 0.0027 0.3225 -0.9466>}
 index_03_r {parent=index_02_r, loc=<-2.5578 0.123 0.0252>, rot=<-0.002 0.0067 0.2558 -0.9667>}
 middle_metacarpal_r {parent=hand_r, loc=<-2.1856 0.1725 0.5984>, rot=<0.006 0.0168 -0.0329 -0.9993>}
 middle_01_r {parent=middle_metacarpal_r, loc=<-7.3275 0.2447 -0.0129>, rot=<-0.0208 0.0042 0.0858 -0.9961>}
 middle_02_r {parent=middle_01_r, loc=<-4.0123 -0.1401 -0.1175>, rot=<-0.0218 0.0258 0.3512 -0.9357>}
 middle_03_r {parent=middle_02_r, loc=<-2.6077 -0.0534 -0.1373>, rot=<0.0082 0.0188 0.1708 -0.9851>}
 ring_metacarpal_r {parent=hand_r, loc=<-2.2074 0.3625 -1.3638>, rot=<-0.0284 0.0722 -0.0458 -0.9959>}
 ring_01_r {parent=ring_metacarpal_r, loc=<-7.0295 0.6197 0.1446>, rot=<-0.0489 -0.0252 0.1892 -0.9804>}
 ring_02_r {parent=ring_01_r, loc=<-3.6174 -0.4647 0.1027>, rot=<0 0 0.2806 -0.9598>}
 ring_03_r {parent=ring_02_r, loc=<-2.4152 -0.114 -0.0199>, rot=<0 0 0.1189 -0.9929>}
 pinky_metacarpal_r {parent=hand_r, loc=<-1.7616 0.9753 -3.3814>, rot=<-0.0397 0.0814 -0.069 -0.9935>}
 pinky_01_r {parent=pinky_metacarpal_r, loc=<-7.1709 0.4831 0.1094>, rot=<-0.0654 -0.0035 0.2242 -0.9723>}
 pinky_02_r {parent=pinky_01_r, loc=<-2.8274 -0.0893 0.1043>, rot=<0.0037 0.0011 0.2854 -0.9584>}
 pinky_03_r {parent=pinky_02_r, loc=<-2.2046 -0.0307 0.0078>, rot=<-0.0002 0.0098 0.1445 -0.9895>}
 dyn_hand_r {parent=hand_r, loc=<-0.0001 -0.0001 0>, rot=<-0.5823 -0.0887 -0.0649 -0.8055>}
 dyn_r_ulna {parent=dyn_hand_r, loc=<0 0 0>, rot=<0 0 0 1>}
 dyn_r_radius {parent=dyn_hand_r, loc=<0 0 0>, rot=<0 0 0 1>}
 lowerarm_twist_02_r {parent=lowerarm_r, loc=<-9.6199 0 0>, rot=<0 0 0 -1>}
 lowerarm_twist_01_r {parent=lowerarm_r, loc=<-17.8549 0 0>, rot=<0 0 0 -1>}
 dyn_elbow_r_lookat {parent=lowerarm_r, loc=<0 0 -0.0001>, rot=<0.9885 0 0 0.1513>}
 upperarm_twist_01_r {parent=upperarm_r, loc=<-0.0041 0 0>, rot=<0 0 0 -1>}
 deform_deltoid_fr_r {parent=upperarm_twist_01_r, loc=<-2.4511 0 -0.0001>, rot=<0 0 0 -1>}
 upperarm_twist_02_r {parent=upperarm_r, loc=<-13.7941 0 0>, rot=<0 0 0 -1>}
 dyn_upperarm_r {parent=clavicle_r, loc=<-15.5111 0 -0.0001>, rot=<0.0008 0.0336 -0.0252 -0.9991>}
 dyn_scapula_r_base {parent=dyn_upperarm_r, loc=<0 0 0>, rot=<0.9708 0.217 0.0994 -0.0222>}
 dyn_scapula_r {parent=dyn_scapula_r_base, loc=<0 0 0>, rot=<0 0 0 1>}
 dyn_necklace_MSTR {parent=spine_04, loc=<13.6647 8.3322 0>, rot=<0.0294 -0.7065 -0.0294 -0.7065>}
 dyn_necklace_01 {parent=dyn_necklace_MSTR, loc=<0 -0.7136 2.8935>, rot=<0.271 -0.271 -0.6531 -0.6531>}
 dyn_necklace_02 {parent=dyn_necklace_01, loc=<6.7046 0 0>, rot=<0 -0.2145 0 -0.9767>}
 dyn_scapula_look_at {parent=spine_04, loc=<10.0435 -9.6719 0>, rot=<0 0 -0.0015 -1>}
 dyn_scapula_l_constraint {parent=dyn_scapula_look_at, loc=<-5.1337 -0.123 -8.2604>, rot=<0.0284 -0.7065 -0.0284 -0.7065>}
 dyn_scapula_r_constraint {parent=dyn_scapula_look_at, loc=<-5.1337 -0.123 8.2604>, rot=<0.0284 -0.7065 -0.0284 -0.7065>}
 thigh_l {parent=pelvis, loc=<-2.5693 -0.0109 -7.7501>, rot=<0.0012 0.0572 0.0155 -0.9982>}
 calf_l {parent=thigh_l, loc=<-39.0484 0 0>, rot=<0.0006 -0.0161 -0.0391 -0.9991>}
 foot_l {parent=calf_l, loc=<-42.561 0.1077 0.4292>, rot=<0.0418 -0.0455 0.0343 -0.9975>}
 ball_l {parent=foot_l, loc=<-6.1599 13.8264 0.6806>, rot=<0.0001 -0.0001 -0.7063 -0.7079>}
 calf_twist_02_l {parent=calf_l, loc=<-18.8064 -0.0967 0.0397>, rot=<0.0237 0.0095 -0.0035 -0.9997>}
 calf_twist_01_l {parent=calf_l, loc=<-35.813 -0.0967 0.0397>, rot=<0.0237 0.0095 -0.0035 -0.9997>}
 dyn_pant_leg_l_01_MSTR {parent=calf_twist_01_l, loc=<3.4021 2.4016 3.6992>, rot=<-0.0439 -0.9973 -0.0266 0.0516>}
 dyn_pant_leg_l_01 {parent=dyn_pant_leg_l_01_MSTR, loc=<0 0 0>, rot=<0 0 0 1>}
 dyn_pant_leg_l_02_MSTR {parent=calf_twist_01_l, loc=<3.0579 1.7692 -4.3961>, rot=<-0.0439 -0.9973 -0.0266 0.0516>}
 dyn_pant_leg_l_02 {parent=dyn_pant_leg_l_02_MSTR, loc=<0 0 0>, rot=<0 0 0 1>}
 thigh_twist_01_l {parent=thigh_l, loc=<-21.6303 0 0>, rot=<0.0013 0 0 -1>}
 deform_thigh_bk_l {parent=thigh_twist_01_l, loc=<16.9514 0.2058 0.1808>, rot=<-0.0013 -0.0217 0 -0.9998>}
 deform_thigh_fr_l {parent=thigh_twist_01_l, loc=<21.5084 0.2281 -0.0598>, rot=<-0.0013 -0.0217 0 -0.9998>}
 thigh_r {parent=pelvis, loc=<-2.5693 -0.0109 7.7501>, rot=<0.0572 -0.0012 0.9982 0.0155>}
 calf_r {parent=thigh_r, loc=<39.0484 0 0>, rot=<0.0006 -0.0161 -0.0391 -0.9991>}
 foot_r {parent=calf_r, loc=<42.561 -0.1077 -0.4292>, rot=<0.0418 -0.0455 0.0343 -0.9975>}
 ball_r {parent=foot_r, loc=<6.1599 -13.8264 -0.6806>, rot=<0.0001 -0.0001 -0.7063 -0.7079>}
 calf_twist_02_r {parent=calf_r, loc=<18.8064 0.0967 -0.0397>, rot=<0.0237 0.0095 -0.0035 -0.9997>}
 calf_twist_01_r {parent=calf_r, loc=<35.813 0.0967 -0.0397>, rot=<0.0237 0.0095 -0.0035 -0.9997>}
 dyn_pant_leg_r_01_MSTR {parent=calf_twist_01_r, loc=<-3.4022 -2.4016 -3.6994>, rot=<0.9973 -0.0439 -0.0516 -0.0266>}
 dyn_pant_leg_r_01 {parent=dyn_pant_leg_r_01_MSTR, loc=<0 0 0>, rot=<0 0 0 1>}
 dyn_pant_leg_r_02_MSTR {parent=calf_twist_01_r, loc=<-3.0579 -1.7692 4.3958>, rot=<0.9973 -0.0439 -0.0516 -0.0266>}
 dyn_pant_leg_r_02 {parent=dyn_pant_leg_r_02_MSTR, loc=<0 0 0>, rot=<0 0 0 1>}
 thigh_twist_01_r {parent=thigh_r, loc=<21.6303 0 0>, rot=<0.0013 0 0 -1>}
 deform_thigh_bk_r {parent=thigh_twist_01_r, loc=<-16.9514 -0.2058 -0.1808>, rot=<-0.0013 -0.0217 0 -0.9998>}
 deform_thigh_fr_r {parent=thigh_twist_01_r, loc=<-21.5085 -0.2281 0.0598>, rot=<-0.0013 -0.0217 0 -0.9998>}
 dyn_belt_MSTR {parent=pelvis, loc=<3.8183 8.413 8.7159>, rot=<0.6459 -0.1725 0.7322 -0.1299>}
 dyn_belt_01 {parent=dyn_belt_MSTR, loc=<0.9115 0 0>, rot=<-0.0051 -0.1896 0.0161 -0.9817>}
 dyn_belt_02 {parent=dyn_belt_01, loc=<5.427 0 0>, rot=<-0.058 -0.3705 0.048 -0.9258>}
 dyn_belt_03 {parent=dyn_belt_02, loc=<6.9322 0 0>, rot=<-0.2278 -0.1632 0.0852 -0.9561>}
 dyn_belt_04 {parent=dyn_belt_03, loc=<7.4526 0 0>, rot=<-0.0618 -0.0027 -0.0441 -0.9971>}
 dyn_root_simspace {parent=root, loc=<0 0 0>, rot=<0 0 0 1>}
]
variables [
 KneeRot_fr_l_114deg {type=FVector, default=(X=0,Y=0,Z=0)}
 KneeRot_fr_l_130deg {type=FVector, default=(X=0,Y=0,Z=0)}
 KneeScale_fr_l_114deg {type=FVector, default=(X=0,Y=0,Z=0)}
 KneeScale_fr_l_130deg {type=FVector, default=(X=0,Y=0,Z=0)}
 KneeTrans_fr_l_114deg {type=FVector, default=(X=0,Y=0,Z=0)}
 KneeTrans_fr_l_130deg {type=FVector, default=(X=0,Y=0,Z=0)}
]
