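# Physics asset for the right-foot mesh of the male medium Keychain cosmetic body, set up around two dynamic pant-leg bones.
# Capsule bodies cover the right calf twist, foot, both pant-leg bones and their tiny MSTR parent bodies.
# Each MSTR body has free angular motion relative to the calf, while each pant bone has narrow twist and uneven swing limits against its MSTR.
# The foot body has no constraint.
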PhysicsAsset /BRCosmetics/Characters/Player/Male/Medium/Bodies/M_MED_Keychain/Meshes/M_MED_KeyChain_r_foot_Physics
bodies {
 calf_twist_01_r [{
   shapeType=Capsule
   length=16.6167
   radius=4.5614
   shapeName=calf_twist_01_r_capsule
   center=<1.6349 0.0001 -0.1107>
   rotation=<rot 85.6662 0.0001 0>
  }]
 dyn_pant_leg_r_01 [{shapeType=Capsule, length=5.3936, radius=3.2051, shapeName=dyn_pant_leg_r_01_capsule, center=<13.6947 0 3.5>, rotation=<rot 0 0 97.8718>}]
 dyn_pant_leg_r_01_MSTR [{shapeType=Capsule, length=0.505, radius=0.505, shapeName=dyn_pant_leg_r_01_MSTR_capsule, rotation=<rot 0 -90 0>}]
 dyn_pant_leg_r_02 [{shapeType=Capsule, length=5.0589, radius=2.8868, shapeName=dyn_pant_leg_r_02_capsule, center=<11.8455 1 -3.5>, rotation=<rot 0 0 90>}]
 dyn_pant_leg_r_02_MSTR [{shapeType=Capsule, length=0.505, radius=0.505, shapeName=dyn_pant_leg_r_02_MSTR_capsule, rotation=<rot 0 -90 0>}]
 foot_r [{shapeType=Capsule, length=18.9494, radius=4.5635, center=<6.3227 -7.2043 0.208>, rotation=<rot 0 0 90.6273>}]
}
constraints [
 {
  bone1Name=dyn_pant_leg_r_01_MSTR
  bone2Name=calf_twist_01_r
  swing1LimitDegrees=45
  swing1Motion=Free
  swing2LimitDegrees=45
  swing2Motion=Free
  twistLimitDegrees=45
  twistMotion=Free
 }
 {
  bone1Name=dyn_pant_leg_r_02_MSTR
  bone2Name=calf_twist_01_r
  swing1LimitDegrees=45
  swing1Motion=Free
  swing2LimitDegrees=45
  swing2Motion=Free
  twistLimitDegrees=45
  twistMotion=Free
 }
 {bone1Name=dyn_pant_leg_r_01, bone2Name=dyn_pant_leg_r_01_MSTR, swing1LimitDegrees=10, swing2LimitDegrees=35, twistLimitDegrees=3}
 {bone1Name=dyn_pant_leg_r_02, bone2Name=dyn_pant_leg_r_02_MSTR, swing1LimitDegrees=1, swing2LimitDegrees=35, twistLimitDegrees=3}
]
