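# Secondary-motion AnimBlueprint for the male medium Keychain body cosmetic, following the attached parent mesh's pose.
# Copied pose runs through twist and sculpter Control Rigs, bone constraints, look-ats and rotation-driven offsets into a cached pre-sim pose.
# That feeds chained RigidBody sims, swapped for fixed belt/necklace offsets when DBNO-carried, then a final Control Rig.
# EventGraph computes offset alphas, runs deprecated gravity and sim-space helpers, and softens main-sim inertia while parachuting.
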AnimBlueprint /BRCosmetics/Characters/Player/Male/Medium/Bodies/M_MED_Keychain/Meshes/M_MED_KeyChain_AnimBP
target_skeleton=&/M_MED_KeyChain_Skeleton
graphs [
 AnimGraph {nodes [
   n1 = Control Rig!(
    Source=Blend Poses by bool(BlendPose_0=Use cached pose 'DBNO_Pose'(), BlendPose_1=Use cached pose 'Sim_Pose'(), bActiveValue=Get bIsDBNOCarried()) {blendType=Linear}
    bAlphaBoolEnabled:bool=True
   ) -> (Pose) {
    alphaInputType=Bool
    alphaBoolBlend {blendOption=Linear}
    controlRigAssetReference {blueprintRigClass=&/M_MED_KeyChain_CtrlRig.M_MED_KeyChain_CtrlRig_C}
   }
   Output Pose(Result=@n1) {layerGroup=None}
   n2 = "Transform (Modify) Bone - Bone: dyn_belt_01"(
    ComponentPose=Local To Component(LocalPose=Use cached pose 'PreSim_Pose'())
    Alpha=1
    Translation="0,0,0"
    Rotation="40,11.2,66"
    Scale="1,1,1"
   ) {
    rotationMode=BMM_Additive
    rotationSpace=BCS_BoneSpace
    translationMode=BMM_Additive
    translationSpace=BCS_BoneSpace
    alphaBoolBlend {blendOption=Linear}
    boneToModify {boneName=dyn_belt_01}
    rotation=<rot 40 11.2 66>
   }
   n3 = "Transform (Modify) Bone - Bone: dyn_belt_02"(ComponentPose=@n2, Alpha=1, Translation="0,0,0", Rotation="52,128,122", Scale="1,1,1") {
    rotationMode=BMM_Additive
    rotationSpace=BCS_BoneSpace
    translationMode=BMM_Additive
    translationSpace=BCS_BoneSpace
    alphaBoolBlend {blendOption=Linear}
    boneToModify {boneName=dyn_belt_02}
    rotation=<rot 52 128 122>
   }
   n4 = "Transform (Modify) Bone - Bone: dyn_belt_03"(ComponentPose=@n3, Alpha=1, Translation="0,0,0", Rotation="33.014,46.247,43.630", Scale="1,1,1") {
    rotationMode=BMM_Additive
    rotationSpace=BCS_BoneSpace
    translationMode=BMM_Additive
    translationSpace=BCS_BoneSpace
    alphaBoolBlend {blendOption=Linear}
    boneToModify {boneName=dyn_belt_03}
    rotation=<rot 33.014 46.247 43.63>
   }
   n5 = "Transform (Modify) Bone - Bone: dyn_belt_04"(ComponentPose=@n4, Alpha=1, Translation="0,0,0", Rotation="37,0.5,51", Scale="1,1,1") {
    rotationMode=BMM_Additive
    rotationSpace=BCS_BoneSpace
    translationMode=BMM_Additive
    translationSpace=BCS_BoneSpace
    alphaBoolBlend {blendOption=Linear}
    boneToModify {boneName=dyn_belt_04}
    rotation=<rot 37 0.5 51>
   }
   n6 = "Transform (Modify) Bone - Bone: dyn_necklace_01"(ComponentPose=@n5, Alpha=1, Translation="0,0,0", Rotation="-0.027,-30,0.018", Scale="1,1,1") {
    rotationMode=BMM_Additive
    rotationSpace=BCS_BoneSpace
    translationMode=BMM_Additive
    translationSpace=BCS_BoneSpace
    alphaBoolBlend {blendOption=Linear}
    boneToModify {boneName=dyn_necklace_01}
    rotation=<rot -0.027 -30 0.018>
   }
   n7 = "Transform (Modify) Bone - Bone: dyn_necklace_02"(ComponentPose=@n6, Alpha=1, Translation="0,0,0", Rotation="3,-15,-0.037", Scale="1,1,1") {
    rotationMode=BMM_Additive
    rotationSpace=BCS_BoneSpace
    translationMode=BMM_Additive
    translationSpace=BCS_BoneSpace
    alphaBoolBlend {blendOption=Linear}
    boneToModify {boneName=dyn_necklace_02}
    rotation=<rot 3 -15 -0.037>
   }
   Save cached pose 'DBNO_Pose'(Pose=Component To Local(ComponentPose=@n7)) {cacheName=DBNO_Pose}
   n8 = Control Rig!(Source=Copy Pose From Mesh() {bUseAttachedParent=true}, Alpha:float) -> (Pose) {
    alphaBoolBlend {blendOption=Linear}
    controlRigAssetReference {blueprintRigClass=&/M_MED_KeyChain_Twists_CtrlRig.M_MED_KeyChain_Twists_CtrlRig_C}
   }
   n9 = Control Rig!(
    Source=@n8
    Alpha
    KneeRot_fr_l_114deg:Vector
    KneeTrans_fr_l_114deg:Vector="0.258,-3.591,0.009"
    KneeScale_fr_l_114deg:Vector="1,1,1"
    KneeScale_fr_l_130deg:Vector="1,1,1"
    KneeRot_fr_l_130deg:Vector="0,0,-51.2"
    KneeTrans_fr_l_130deg:Vector="0.258,-3.591,0.009"
   ) -> (Pose) {
    alphaBoolBlend {blendOption=Linear}
    controlRigAssetReference {blueprintRigClass=&/M_MED_KeyChain_Sculpters_2_CtrlRig.M_MED_KeyChain_Sculpters_2_CtrlRig_C}
   }
   n10 = "Constraint - dyn_necklace_MSTR (PPPPPPP)"!(
    ComponentPose=Local To Component(LocalPose=@n9)
    Alpha=1
    ConstraintWeights_0=0.042
    ConstraintWeights_1:float=0.033
    ConstraintWeights_2:float=0.025
    ConstraintWeights_3:float=0.04
    ConstraintWeights_4:float=0.78
    ConstraintWeights_5:float=0.04
    ConstraintWeights_6:float=0.04
   ) -> (Pose) {
    alphaBoolBlend {blendOption=Linear}
    boneToModify {boneName=dyn_necklace_MSTR}
    constraintSetup [
     {targetBone {boneName=neck_02}}
     {targetBone {boneName=neck_01}}
     {targetBone {boneName=spine_03}}
     {targetBone {boneName=spine_04}}
     {targetBone {boneName=spine_05}}
     {targetBone {boneName=clavicle_l}}
     {targetBone {boneName=clavicle_r}}
    ]
    constraintWeights [1,1,1,1,1,1,1]
   }
   n11 = "Constraint - dyn_belt_MSTR (PPPPPP)"!(
    ComponentPose=@n10
    Alpha=1
    ConstraintWeights_0=0.0895
    ConstraintWeights_1:float=0.2115
    ConstraintWeights_2:float=0.0103
    ConstraintWeights_3:float=0.0036
    ConstraintWeights_4:float=0.6783
    ConstraintWeights_5:float=0.0068
   ) -> (Pose) {
    alphaBoolBlend {blendOption=Linear}
    boneToModify {boneName=dyn_belt_MSTR}
    constraintSetup [
     {targetBone {boneName=thigh_twist_01_r}}
     {targetBone {boneName=spine_01}}
     {targetBone {boneName=spine_02}}
     {targetBone {boneName=spine_03}}
     {targetBone {boneName=pelvis}}
     {targetBone {boneName=thigh_twist_01_l}}
    ]
    constraintWeights [1,1,1,1,1,1]
   }
   n12 = "Constraint - dyn_pant_leg_r_02_MSTR (PP)"!(ComponentPose=@n11, Alpha=1, ConstraintWeights_0=0.3159, ConstraintWeights_1:float=0.6841) -> (Pose) {
    alphaBoolBlend {blendOption=Linear}
    boneToModify {boneName=dyn_pant_leg_r_02_MSTR}
    constraintSetup [{targetBone {boneName=calf_twist_02_r}}, {targetBone {boneName=calf_twist_01_r}}]
    constraintWeights [1,1]
   }
   n13 = "Constraint - dyn_pant_leg_r_01_MSTR (PP)"!(ComponentPose=@n12, Alpha=1, ConstraintWeights_0=0.3989, ConstraintWeights_1:float=0.6011) -> (Pose) {
    alphaBoolBlend {blendOption=Linear}
    boneToModify {boneName=dyn_pant_leg_r_01_MSTR}
    constraintSetup [{targetBone {boneName=calf_twist_02_r}}, {targetBone {boneName=calf_twist_01_r}}]
    constraintWeights [1,1]
   }
   n14 = "Constraint - dyn_pant_leg_l_01_MSTR (PP)"!(ComponentPose=@n13, Alpha=1, ConstraintWeights_0=0.6011, ConstraintWeights_1:float=0.3989) -> (Pose) {
    alphaBoolBlend {blendOption=Linear}
    boneToModify {boneName=dyn_pant_leg_l_01_MSTR}
    constraintSetup [{targetBone {boneName=calf_twist_01_l}}, {targetBone {boneName=calf_twist_02_l}}]
    constraintWeights [1,1]
   }
   n15 = "Constraint - dyn_pant_leg_l_02_MSTR (PP)"!(ComponentPose=@n14, Alpha=1, ConstraintWeights_0=0.6841, ConstraintWeights_1:float=0.3159) -> (Pose) {
    alphaBoolBlend {blendOption=Linear}
    boneToModify {boneName=dyn_pant_leg_l_02_MSTR}
    constraintSetup [{targetBone {boneName=calf_twist_01_l}}, {targetBone {boneName=calf_twist_02_l}}]
    constraintWeights [1,1]
   }
   n16 = "Constraint - dyn_scapula_look_at (PPPPPPP)"!(
    ComponentPose=@n15
    Alpha=1
    ConstraintWeights_0=0.0027
    ConstraintWeights_1:float=0.046
    ConstraintWeights_2:float=0.1842
    ConstraintWeights_3:float=0.2081
    ConstraintWeights_4:float=0.5053
    ConstraintWeights_5:float=0.0345
    ConstraintWeights_6:float=0.0192
   ) -> (Pose) {
    alphaBoolBlend {blendOption=Linear}
    boneToModify {boneName=dyn_scapula_look_at}
    constraintSetup [
     {targetBone {boneName=spine_01}}
     {targetBone {boneName=spine_02}}
     {targetBone {boneName=spine_03}}
     {targetBone {boneName=spine_04}}
     {targetBone {boneName=spine_05}}
     {targetBone {boneName=neck_02}}
     {targetBone {boneName=neck_01}}
    ]
    constraintWeights [1,1,1,1,1,1,1]
   }
   n17 = "Constraint - dyn_radius_lookat (P)"(ComponentPose=@n16, Alpha=1, ConstraintWeights_0=1) {
    alphaBoolBlend {blendOption=Linear}
    boneToModify {boneName=dyn_radius_lookat}
    constraintSetup [{targetBone {boneName=lowerarm_l}}]
    constraintWeights [1]
   }
   n18 = Look At - Bone: dyn_radius(ComponentPose=@n17, Alpha=1) {
    alphaBoolBlend {blendOption=Linear}
    boneToModify {boneName=dyn_radius}
    lookAtTarget {boneReference {boneName=dyn_radius_lookat}}
    lookAt_Axis {axis=<1 0>}
   }
   n19 = Look At - Bone: dyn_ulna(ComponentPose=@n18, Alpha=1) {
    alphaBoolBlend {blendOption=Linear}
    boneToModify {boneName=dyn_ulna}
    lookAtTarget {boneReference {boneName=dyn_radius_lookat}}
    lookAt_Axis {axis=<1 0>}
   }
   n20 = "Constraint - dyn_elbow_r_lookat (P)"(ComponentPose=@n19, Alpha=1, ConstraintWeights_0=1) {
    alphaBoolBlend {blendOption=Linear}
    boneToModify {boneName=dyn_elbow_r_lookat}
    constraintSetup [{targetBone {boneName=lowerarm_r}}]
    constraintWeights [1]
   }
   n21 = Look At - Bone: dyn_r_radius(ComponentPose=@n20, Alpha=1) {
    alphaBoolBlend {blendOption=Linear}
    boneToModify {boneName=dyn_r_radius}
    lookAtTarget {boneReference {boneName=dyn_elbow_r_lookat}}
    lookAt_Axis {axis=<1 0>}
   }
   n22 = Look At - Bone: dyn_r_ulna(ComponentPose=@n21, Alpha=1) {
    alphaBoolBlend {blendOption=Linear}
    boneToModify {boneName=dyn_r_ulna}
    lookAtTarget {boneReference {boneName=dyn_elbow_r_lookat}}
    lookAt_Axis {axis=<1 0>}
   }
   n23 = "Transform (Modify) Bone - Bone: dyn_neck_01"(ComponentPose=@n22, Alpha=Get Neck_01 rot() -> (Neck_01 rot:double), Translation="0.000,-5,0.000", Rotation="0,0,0", Scale="1,1,1") {translationMode=BMM_Additive, alphaBoolBlend {blendOption=Linear}, boneToModify {boneName=dyn_neck_01}, translation {y=-5}}
   n24 = "Transform (Modify) Bone - Bone: dyn_belt_MSTR"(
    ComponentPose=@n23
    Alpha=Get Alpha_Thigh_R_Fr() -> (Alpha_Thigh_R_Fr:double)
    Translation="-0.795,-0.689,5.063"
    Rotation="0,0,0"
    Scale="1,1,1"
   ) {translationMode=BMM_Additive, alphaBoolBlend {blendOption=Linear}, boneToModify {boneName=dyn_belt_MSTR}, translation=<-0.795 -0.689 5.063>}
   n25 = Look At - Bone: dyn_scapula_l_base(ComponentPose=@n24, Alpha=1) {
    alphaBoolBlend {blendOption=Linear}
    boneToModify {boneName=dyn_scapula_l_base}
    lookAtLocation {x=0}
    lookAtTarget {boneReference {boneName=dyn_scapula_look_at}}
    lookAt_Axis {axis=<1 0>}
   }
   n26 = Look At - Bone: dyn_scapula_r_base(ComponentPose=@n25, Alpha=1) {
    alphaBoolBlend {blendOption=Linear}
    boneToModify {boneName=dyn_scapula_r_base}
    lookAtLocation {x=0}
    lookAtTarget {boneReference {boneName=dyn_scapula_look_at}}
    lookAt_Axis {axis=<1 0>}
   }
   Save cached pose 'PreSim_Pose'(Pose=Component To Local(ComponentPose=@n26)) {cacheName=PreSim_Pose}
   n27 = RigidBody!(ComponentPose=Local To Component(LocalPose=Use cached pose 'PreSim_Pose'()), Alpha, OverrideWorldGravity:Vector="0,0,-980", ExternalForce) -> (Pose) {
    bClampLinearTranslationLimitToRefPose=true
    bOverrideWorldGravity=true
    overridePhysicsAsset=&/M_MED_KeyChain_Scapula_Physics
    simulationSpace=BaseBoneSpace
    alphaBoolBlend {blendOption=Linear}
    baseBoneRef {boneName=spine_05}
    componentAppliedLinearAccClamp=<100 100 100>
    componentLinearAccScale=<0.2 0.2 0.2>
    componentLinearVelScale=<0.1 0.1 0.1>
    overrideWorldGravity {z=-980}
   }
   n28 = RigidBody!(
    ComponentPose=@n27
    Alpha
    OverrideWorldGravity="0,0,-980"
    ExternalForce
    ComponentLinearAccScale:Vector=Get Linear Acc() -> (Linear Acc:Vector)
    ComponentLinearVelScale:Vector=Get Linear Vel() -> (Linear Vel:Vector)
   ) -> (Pose) {
    bClampLinearTranslationLimitToRefPose=true
    bOverrideWorldGravity=true
    overridePhysicsAsset=&/M_MED_KeyChain_Physics
    simulationSpace=BaseBoneSpace
    alphaBoolBlend {blendOption=Linear}
    baseBoneRef {boneName=dyn_root_simspace}
    componentAppliedLinearAccClamp=<3000 3000 3000>
    componentLinearAccScale=<0.3 0.3 0.5>
    componentLinearVelScale=<0.2 0.2 0.3>
    overrideWorldGravity {z=-980}
   }
   n29 = RigidBody!(ComponentPose=@n28, Alpha, OverrideWorldGravity="0,0,-980", ExternalForce) -> (Pose) {
    bClampLinearTranslationLimitToRefPose=true
    bOverrideWorldGravity=true
    overridePhysicsAsset=&/M_MED_KeyChain_Necklace_Physics
    simulationSpace=BaseBoneSpace
    alphaBoolBlend {blendOption=Linear}
    baseBoneRef {boneName=dyn_root_simspace}
    componentAppliedLinearAccClamp=<3000 3000 3000>
    componentLinearAccScale=<0.3 0.3 0.5>
    componentLinearVelScale=<0.3 0.3 0.3>
    overrideWorldGravity {z=-980}
   }
   n30 = RigidBody(ComponentPose=@n29, OverrideWorldGravity="0,0,-980") {
    bClampLinearTranslationLimitToRefPose=true
    bOverrideWorldGravity=true
    overridePhysicsAsset=&/M_MED_KeyChain_l_foot_Physics
    simulationSpace=BaseBoneSpace
    alphaBoolBlend {blendOption=Linear}
    baseBoneRef {boneName=calf_l}
    componentLinearAccScale=<0.5 0.5 0.5>
    componentLinearVelScale=<0.3 0.3 0.3>
    overrideWorldGravity {z=-980}
   }
   n31 = RigidBody(ComponentPose=@n30, OverrideWorldGravity="0,0,-980") {
    bClampLinearTranslationLimitToRefPose=true
    bOverrideWorldGravity=true
    overridePhysicsAsset=&/M_MED_KeyChain_r_foot_Physics
    simulationSpace=BaseBoneSpace
    alphaBoolBlend {blendOption=Linear}
    baseBoneRef {boneName=calf_r}
    componentLinearAccScale=<0.5 0.5 0.5>
    componentLinearVelScale=<0.3 0.3 0.3>
    overrideWorldGravity {z=-980}
   }
   Save cached pose 'Sim_Pose'(Pose=Component To Local(ComponentPose=@n31)) {cacheName=Sim_Pose}
  ]}
 EventGraph {nodes [
   Event BlueprintUpdateAnimation()
   > Get!(execute:Exec) -> (then:Exec, else:Exec, Skel_Mesh:Skeletal Mesh Component Object Reference)
   > n1 = Sequence!(execute) -> (then_0, then_1, then_2:Exec, then_3:Exec)
   > n2 = Gravity Override State Function DEPRECATED(CharacterPartAnimInstance=Self-Reference(), GameTimeInSeconds=GetGameTimeInSeconds(), GravityOverrideParams=Get Gravity Override Params())
   > Set GravityOverride(GravityOverride=@n2)
   n3 = Get Skel_Mesh()
   Set Neck_01 rot(
    execute=@n1.then_1
    Neck_01 rot:double=GetAlphaFromReferenceTransformDelta(SkeletalMeshComponent=@n3, BoneName=head, BaseName=spine_05, InputRange=(X=-15,Y=20), TwistAxis="1,0,0")
   ) -> (Output_Get:double)
   > Set Alpha_Thigh_R_Fr(Alpha_Thigh_R_Fr:double=GetAlphaFromReferenceTransformDelta(SkeletalMeshComponent=@n3, BoneName=thigh_r, BaseName=pelvis, InputRange=(X=-10,Y=-100), TwistAxis="1,0,0")) -> (Output_Get:double)
   n4 = ComplexSimSpaceFunction DEPRECATED!(
    execute=@n1.then_2
    CharacterPartAnimInstance:Custom Character Part Anim Instance Object Reference=Self-Reference()
    InterpSpeedMax:double=9
    PawnSpeedMax:double=20
    RInterpErrorTolerance:double=1
    DeltaTime:double=0
    DebugDraw:bool=false
    sim_space_rotation_in:Rotator=Get sim_space_rotation() -> (sim_space_rotation:Rotator)
    inComponentSpaceIn:bool=Get InComponentSpace() -> (InComponentSpace:bool)
    inComponentSpaceLastFrameIn:bool=Get InComponentSpaceLastFrame() -> (InComponentSpaceLastFrame:bool)
    lockedOnSimSpaceIn:bool=Get LockedOnSimSpace() -> (LockedOnSimSpace:bool)
    InterpSpeedIn:double=Get InterpSpeed() -> (InterpSpeed:double)
   ) -> (
    then
    sim_space_rotation:Rotator
    inComponentSpaceOut:bool
    inComponentSpaceLastFrameOut:bool
    lockedOnSimSpaceOut:bool
    InterpSpeedOut:double
   )
   > Set sim_space_rotation(sim_space_rotation:Rotator=@n4) -> (Output_Get:Rotator)
   > Set InComponentSpace(InComponentSpace:bool=@n4.inComponentSpaceOut) -> (Output_Get:bool)
   > Set InComponentSpaceLastFrame(InComponentSpaceLastFrame:bool=@n4.inComponentSpaceLastFrameOut) -> (Output_Get:bool)
   > Set LockedOnSimSpace(LockedOnSimSpace:bool=@n4.lockedOnSimSpaceOut) -> (Output_Get:bool)
   > Set InterpSpeed(InterpSpeed:double=@n4.InterpSpeedOut) -> (Output_Get:double)
   n5 = Branch(execute=@n1.then_3, Condition=Get bIsParachuteOpened())
   > Set Linear Acc(Linear Acc:Vector="0.2,0.2,0.3") -> (Output_Get:Vector)
   > Set Linear Vel(Linear Vel:Vector="0.1,0.1,0.2") -> (Output_Get:Vector)
   Set Linear Acc(execute=@n5.else, Linear Acc="0.3,0.3,0.5")
   > Set Linear Vel(Linear Vel="0.2,0.2,0.3")
   Event BlueprintInitializeAnimation()
   > Set Skel_Mesh(Skel_Mesh=GetOwnerSkeletalMeshComponent())
  ]}
]
variables [
 Alpha_Thigh_R_Fr {type=number}
 Gravity Override Params {type=GravityOverrideParamsStruct}
 GravityOverride {type=Vector}
 InComponentSpace {type=boolean}
 InComponentSpaceLastFrame {type=boolean}
 InterpSpeed {type=number}
 Linear Acc {type=Vector}
 Linear Vel {type=Vector}
 LockedOnSimSpace {type=boolean}
 Neck_01 rot {type=number}
 sim_space_rotation {type=Rotator}
 Skel_Mesh {type=SkeletalMeshComponent}
]
